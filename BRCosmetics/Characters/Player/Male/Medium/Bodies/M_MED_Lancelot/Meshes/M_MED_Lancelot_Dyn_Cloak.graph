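# Physics asset for the dynamic cloak on the medium male Lancelot player body.
# Three cape bone chains (right, mid-right and shorter mid-left) hang from MSTR bones, with limited swing and tight twist, cross-linked in places by free constraints.
# Capsules on pelvis, spine, arms, right hand, legs and feet form body colliders, mostly joined by free constraints.
# The dyn_root_slope bone carries two spheres, one huge and offset downward like a curved floor.
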PhysicsAsset /BRCosmetics/Characters/Player/Male/Medium/Bodies/M_MED_Lancelot/Meshes/M_MED_Lancelot_Dyn_Cloak
bodies {
 calf_l [{
   shapeType=Capsule
   length=34.0334
   radius=9.5628
   shapeName=calf_r_capsule
   center=<-21.8366 -0.1015 0>
   rotation=<rot -90 -169.3803 -13.8166>
  }]
 calf_r [{shapeType=Capsule, length=34.0334, radius=9.5628, shapeName=calf_r_capsule, center=<21.8366 0.1015 0>, rotation=<rot -90 -30.6997 27.5028>}]
 cape_01_r [{shapeType=Capsule, length=4.5, radius=12.5, shapeName=cape_01_r_capsule, center=<5 0 10>, rotation=<rot 90 0 0>}]
 cape_02_r [{shapeType=Capsule, length=9.5, radius=12.5, shapeName=cape_02_r_capsule, center=<6 5 10>, rotation=<rot -0.0001 0 85>}]
 cape_03_r [{shapeType=Capsule, length=10.5, radius=16, shapeName=cape_03_r_capsule, center=<10 5 16>, rotation=<rot -0.0004 0 80>}]
 cape_04_r [{shapeType=Capsule, length=10.5, radius=16, shapeName=cape_04_r_capsule, center=<10 5 16>, rotation=<rot -0.0004 0 80>}]
 cape_05_r [{shapeType=Capsule, length=9.5, radius=17, shapeName=cape_05_r_capsule, center=<10 5 16>, rotation=<rot -0.0004 0 80>}]
 cape_mid_01_l [{shapeType=Capsule, length=5, radius=12, shapeName=cape_mid_01_r_capsule, center=<5 0 10>, rotation=<rot 0 0 90>}]
 cape_mid_01_r [{shapeType=Capsule, length=4.5, radius=12.5, shapeName=cape_mid_01_r_capsule, center=<5 0 10>, rotation=<rot 0 0 90>}]
 cape_mid_02_l [{shapeType=Capsule, length=13.25, radius=16.375, shapeName=cape_mid_02_r_capsule, center=<10 -5 14>, rotation=<rot 0 0 90>}]
 cape_mid_02_r [{shapeType=Capsule, length=9.75, radius=12.25, shapeName=cape_mid_02_r_capsule, center=<6 0 10>, rotation=<rot 0 0 90>}]
 cape_mid_03_l [{shapeType=Capsule, length=11.75, radius=17.875, shapeName=cape_mid_03_r_capsule, center=<10 -5 14>, rotation=<rot 0 0 90>}]
 cape_mid_03_r [{shapeType=Capsule, length=11.75, radius=17.875, shapeName=cape_mid_03_r_capsule, center=<10 0 15>, rotation=<rot 0 0 90>}]
 cape_mid_04_r [{shapeType=Capsule, length=11.75, radius=17.875, shapeName=cape_mid_04_r_capsule, center=<10 0 15>, rotation=<rot 0 0 90>}]
 cape_mid_05_r [{shapeType=Capsule, length=12.875, radius=16.75, shapeName=cape_mid_05_r_capsule, center=<10 0 12>, rotation=<rot 0 0 90>}]
 cape_mid_l_MSTR [{shapeType=Capsule, length=1, radius=1, shapeName=cape_mid_l_MSTR_capsule, rotation=<rot 90 0 0>}]
 cape_mid_r_MSTR [{shapeType=Capsule, length=1, radius=1, shapeName=cape_mid_r_MSTR_capsule, rotation=<rot 90 0 0>}]
 cape_r_MSTR [{shapeType=Capsule, length=1, radius=1, shapeName=cape_mid_l_MSTR_capsule, rotation=<rot 90 0 0>}]
 dyn_root_slope [
  {shapeType=Sphere, radius=300, shapeName=dyn_root_slope_sphere, center=<0 -40.4097 -292.7278>}
  {shapeType=Sphere, radius=30.25, shapeName=dyn_root_slope_sphere, center=<0 -3.7802 0>}
 ]
 foot_l [{shapeType=Capsule, length=12.7813, radius=11.0452, shapeName=foot_r_capsule, center=<-4.831 7.2726 0>, rotation=<rot 90 -165.9638 -75.9638>}]
 foot_r [{shapeType=Capsule, length=12.7813, radius=11.0452, shapeName=foot_r_capsule, center=<4.831 -7.2726 0>, rotation=<rot 90 180 90>}]
 hand_r [{
   shapeType=Capsule
   length=17.2849
   radius=8.75
   shapeName=lowerarm_r_capsule
   center=<-5.9516 2.1324 -5.1219>
   rotation=<rot 19.9998 0 -0.0002>
  }]
 lowerarm_l [{shapeType=Capsule, length=27.4099, radius=6.625, shapeName=lowerarm_r_capsule, center=<10.0321 2.7329 0>, rotation=<rot -90 180 0>}]
 lowerarm_r [{shapeType=Capsule, length=27.4099, radius=6.625, shapeName=lowerarm_r_capsule, center=<-10.0321 -2.7329 0>, rotation=<rot -90 0 0>}]
 pelvis [{shapeType=Capsule, length=9.2807, radius=14.869, shapeName=pelvis_capsule, center=<-5.8722 0.6785 0>}]
 spine_01 [{shapeType=Capsule, length=7.8323, radius=14.4424, shapeName=spine_01_capsule, center=<-2.5245 0.3335 0>}]
 spine_02 [{shapeType=Capsule, length=7.4605, radius=12.8382, shapeName=spine_02_capsule}]
 spine_03 [{shapeType=Capsule, length=6.5855, radius=11.8382, shapeName=spine_02_capsule}]
 thigh_l [{shapeType=Capsule, length=22.4725, radius=12.8748, shapeName=thigh_l_capsule, center=<-15.2546 0.0105 0>, rotation=<rot 90 32.0054 37.3035>}]
 thigh_r [{shapeType=Capsule, length=22.4725, radius=12.8748, shapeName=thigh_l_capsule, center=<15.2546 -0.0105 0>, rotation=<rot 90 162.646 -12.0559>}]
 upperarm_l [{
   shapeType=Capsule
   length=24.3802
   radius=6.25
   shapeName=upperarm_r_capsule
   center=<11.3453 0.1826 -1.3767>
   rotation=<rot -85 0.001 179.999>
  }]
 upperarm_r [{
   shapeType=Capsule
   length=24.3802
   radius=6.25
   shapeName=upperarm_r_capsule
   center=<-11.3453 -0.1826 1.3767>
   rotation=<rot -85 0.001 -0.001>
  }]
}
constraints [
 {bone1Name=cape_mid_01_r, bone2Name=cape_mid_r_MSTR, swing1LimitDegrees=25, swing2LimitDegrees=35, twistLimitDegrees=5}
 {bone1Name=cape_mid_02_r, bone2Name=cape_mid_01_r, swing1LimitDegrees=25, swing2LimitDegrees=35, twistLimitDegrees=5}
 {bone1Name=cape_mid_03_r, bone2Name=cape_mid_02_r, swing1LimitDegrees=25, swing2LimitDegrees=35, twistLimitDegrees=5}
 {bone1Name=cape_mid_04_r, bone2Name=cape_mid_03_r, swing1LimitDegrees=25, swing2LimitDegrees=35, twistLimitDegrees=5}
 {bone1Name=cape_mid_05_r, bone2Name=cape_mid_04_r, swing1LimitDegrees=25, swing2LimitDegrees=35, twistLimitDegrees=5}
 {bone1Name=cape_02_r, bone2Name=cape_01_r, swing1LimitDegrees=25, swing2LimitDegrees=35, twistLimitDegrees=5}
 {bone1Name=cape_03_r, bone2Name=cape_02_r, swing1LimitDegrees=25, swing2LimitDegrees=35, twistLimitDegrees=5}
 {bone1Name=cape_04_r, bone2Name=cape_03_r, swing1LimitDegrees=25, swing2LimitDegrees=35, twistLimitDegrees=5}
 {bone1Name=cape_05_r, bone2Name=cape_04_r, swing1LimitDegrees=25, swing2LimitDegrees=35, twistLimitDegrees=5}
 {bone1Name=cape_mid_01_l, bone2Name=cape_mid_l_MSTR, swing1LimitDegrees=45, swing2LimitDegrees=65, twistLimitDegrees=5}
 {bone1Name=cape_mid_02_l, bone2Name=cape_mid_01_l, swing1LimitDegrees=45, swing2LimitDegrees=65, twistLimitDegrees=5}
 {bone1Name=cape_mid_03_l, bone2Name=cape_mid_02_l, swing1LimitDegrees=45, swing2LimitDegrees=65, twistLimitDegrees=5}
 {
  bone1Name=spine_01
  bone2Name=pelvis
  swing1LimitDegrees=45
  swing1Motion=Free
  swing2LimitDegrees=45
  swing2Motion=Free
  twistLimitDegrees=45
  twistMotion=Free
 }
 {
  bone1Name=spine_02
  bone2Name=spine_01
  swing1LimitDegrees=45
  swing1Motion=Free
  swing2LimitDegrees=45
  swing2Motion=Free
  twistLimitDegrees=45
  twistMotion=Free
 }
 {
  bone1Name=upperarm_r
  bone2Name=spine_03
  swing1LimitDegrees=45
  swing1Motion=Free
  swing2LimitDegrees=45
  swing2Motion=Free
  twistLimitDegrees=45
  twistMotion=Free
 }
 {
  bone1Name=thigh_l
  bone2Name=pelvis
  swing1LimitDegrees=45
  swing1Motion=Free
  swing2LimitDegrees=45
  swing2Motion=Free
  twistLimitDegrees=45
  twistMotion=Free
 }
 {
  bone1Name=calf_r
  bone2Name=thigh_r
  swing1LimitDegrees=45
  swing1Motion=Free
  swing2LimitDegrees=45
  swing2Motion=Free
  twistLimitDegrees=45
  twistMotion=Free
 }
 {
  bone1Name=foot_r
  bone2Name=calf_r
  swing1LimitDegrees=45
  swing1Motion=Free
  swing2LimitDegrees=45
  swing2Motion=Free
  twistLimitDegrees=45
  twistMotion=Free
 }
 {
  bone1Name=lowerarm_r
  bone2Name=upperarm_r
  swing1LimitDegrees=45
  swing1Motion=Free
  swing2LimitDegrees=45
  swing2Motion=Free
  twistLimitDegrees=45
  twistMotion=Free
 }
 {bone1Name=cape_01_r, bone2Name=cape_r_MSTR, swing1LimitDegrees=25, swing2LimitDegrees=35, twistLimitDegrees=5}
 {
  bone1Name=upperarm_l
  bone2Name=spine_03
  swing1LimitDegrees=45
  swing1Motion=Free
  swing2LimitDegrees=45
  swing2Motion=Free
  twistLimitDegrees=45
  twistMotion=Free
 }
 {bone1Name=lowerarm_l, bone2Name=upperarm_l, swing1LimitDegrees=35, swing2LimitDegrees=35, twistLimitDegrees=15}
 {
  bone1Name=calf_l
  bone2Name=thigh_l
  swing1LimitDegrees=45
  swing1Motion=Free
  swing2LimitDegrees=45
  swing2Motion=Free
  twistLimitDegrees=45
  twistMotion=Free
 }
 {
  bone1Name=foot_l
  bone2Name=calf_l
  swing1LimitDegrees=45
  swing1Motion=Free
  swing2LimitDegrees=45
  swing2Motion=Free
  twistLimitDegrees=45
  twistMotion=Free
 }
 {
  bone1Name=hand_r
  bone2Name=lowerarm_r
  swing1LimitDegrees=45
  swing1Motion=Free
  swing2LimitDegrees=45
  swing2Motion=Free
  twistLimitDegrees=45
  twistMotion=Free
 }
 {
  bone1Name=cape_mid_04_r
  bone2Name=cape_mid_03_l
  swing1LimitDegrees=45
  swing1Motion=Free
  swing2LimitDegrees=45
  swing2Motion=Free
  twistLimitDegrees=45
  twistMotion=Free
 }
 {
  bone1Name=cape_r_MSTR
  bone2Name=spine_03
  swing1LimitDegrees=45
  swing1Motion=Free
  swing2LimitDegrees=45
  swing2Motion=Free
  twistLimitDegrees=45
  twistMotion=Free
 }
 {
  bone1Name=cape_mid_r_MSTR
  bone2Name=spine_03
  swing1LimitDegrees=45
  swing1Motion=Free
  swing2LimitDegrees=45
  swing2Motion=Free
  twistLimitDegrees=45
  twistMotion=Free
 }
 {
  bone1Name=spine_03
  bone2Name=spine_02
  swing1LimitDegrees=45
  swing1Motion=Free
  swing2LimitDegrees=45
  swing2Motion=Free
  twistLimitDegrees=45
  twistMotion=Free
 }
 {
  bone1Name=thigh_r
  bone2Name=pelvis
  swing1LimitDegrees=45
  swing1Motion=Free
  swing2LimitDegrees=45
  swing2Motion=Free
  twistLimitDegrees=45
  twistMotion=Free
 }
 {
  bone1Name=cape_05_r
  bone2Name=cape_mid_05_r
  swing1LimitDegrees=10
  swing1Motion=Free
  swing2LimitDegrees=45
  swing2Motion=Free
  twistLimitDegrees=45
  twistMotion=Free
 }
 {
  bone1Name=cape_03_r
  bone2Name=cape_mid_03_r
  swing1LimitDegrees=10
  swing1Motion=Free
  swing2LimitDegrees=45
  swing2Motion=Free
  twistLimitDegrees=45
  twistMotion=Free
 }
]
