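# Physics asset for the Lancelot medium male character's dynamic cloak, named as the pack-wearing variant for cloth.
# Capsules cover the spine, pelvis, legs, feet, arms and right hand, linked mostly by free angular constraints.
# Three asymmetric cape chains (right, mid-right, shorter mid-left) hang from tiny MSTR anchors on the upper spine.
# Large offset capsules along each chain use narrow swing and twist limits.
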PhysicsAsset /BRCosmetics/Characters/Player/Male/Medium/Bodies/M_MED_Lancelot/Meshes/M_MED_Lancelot_Dyn_Cloak_Wearing_Pack_For_Cloth
bodies {
 calf_l [{
   shapeType=Capsule
   length=34.0334
   radius=9.5628
   shapeName=calf_r_capsule
   center=<-21.8366 -0.1015 0>
   rotation=<rot -90 -169.3803 -13.8166>
  }]
 calf_r [{shapeType=Capsule, length=34.0334, radius=9.5628, shapeName=calf_r_capsule, center=<21.8366 0.1015 0>, rotation=<rot -90 -30.6997 27.5028>}]
 cape_01_r [{shapeType=Capsule, length=4.5, radius=10, shapeName=cape_01_r_capsule, center=<6 5 10>, rotation=<rot 0 0 80>}]
 cape_02_r [{shapeType=Capsule, length=9.5, radius=10, shapeName=cape_02_r_capsule, center=<6 5 10>, rotation=<rot 0 0 80>}]
 cape_03_r [{shapeType=Capsule, length=9.5, radius=10, shapeName=cape_02_r_capsule, center=<6 5 10>, rotation=<rot 0 0 80>}]
 cape_mid_01_l [{shapeType=Capsule, length=9.75, radius=12.25, shapeName=cape_mid_02_r_capsule, center=<6 -5 10>, rotation=<rot 0 0 90>}]
 cape_mid_01_r [{shapeType=Capsule, length=13.25, radius=11.75, shapeName=cape_mid_01_r_capsule, center=<7 5 10>, rotation=<rot 0 0 90>}]
 cape_mid_02_l [{shapeType=Capsule, length=9.75, radius=12.25, shapeName=cape_mid_02_r_capsule, center=<6 -5 10>, rotation=<rot 0 0 90>}]
 cape_mid_02_r [{shapeType=Capsule, length=17.75, radius=12.25, shapeName=cape_mid_02_r_capsule, center=<6 5 10>, rotation=<rot 0 0 90>}]
 cape_mid_03_r [{shapeType=Capsule, length=17.75, radius=12.25, shapeName=cape_mid_02_r_capsule, center=<6 5 10>, rotation=<rot 0 0 90>}]
 cape_mid_l_MSTR [{shapeType=Capsule, length=1, radius=1, shapeName=cape_mid_r_MSTR_capsule, rotation=<rot 90 0 0>}]
 cape_mid_r_MSTR [{shapeType=Capsule, length=1, radius=1, shapeName=cape_mid_r_MSTR_capsule, rotation=<rot 90 0 0>}]
 cape_r_MSTR [{shapeType=Capsule, length=1, radius=1, shapeName=cape_mid_l_MSTR_capsule, rotation=<rot 90 0 0>}]
 foot_l [{shapeType=Capsule, length=9.1563, radius=9.9202, shapeName=foot_r_capsule, center=<-4.831 7.2726 0>, rotation=<rot 90 -165.9638 -75.9638>}]
 foot_r [{shapeType=Capsule, length=9.1563, radius=9.9202, shapeName=foot_r_capsule, center=<4.831 -7.2726 0>, rotation=<rot 90 180 90>}]
 hand_r [{
   shapeType=Capsule
   length=17.2849
   radius=8.75
   shapeName=lowerarm_r_capsule
   center=<-5.9516 2.1324 -5.1219>
   rotation=<rot 19.9998 0 -0.0002>
  }]
 lowerarm_l [{shapeType=Capsule, length=27.4099, radius=6.625, shapeName=lowerarm_r_capsule, center=<10.0321 2.7329 0>, rotation=<rot -90 180 0>}]
 lowerarm_r [{shapeType=Capsule, length=27.4099, radius=6.625, shapeName=lowerarm_r_capsule, center=<-10.0321 -2.7329 0>, rotation=<rot -90 0 0>}]
 pelvis [{shapeType=Capsule, length=4.4057, radius=14.244, shapeName=pelvis_capsule, center=<-5.8722 0.6785 0>}]
 spine_01 [{shapeType=Capsule, length=4.8323, radius=13.8174, shapeName=spine_01_capsule, center=<-2.5245 0.3335 0>}]
 spine_02 [{shapeType=Capsule, length=6.3355, radius=12.0882, shapeName=spine_02_capsule}]
 spine_03 [{shapeType=Capsule, length=7.0855, radius=11.3382, shapeName=spine_02_capsule}]
 thigh_l [{shapeType=Capsule, length=23.0975, radius=12.2498, shapeName=thigh_l_capsule, center=<-15.2556 0 0>, rotation=<rot 90 89.2957 89.9281>}]
 thigh_r [{shapeType=Capsule, length=23.0975, radius=12.2498, shapeName=thigh_l_capsule, center=<15.2556 0 0>, rotation=<rot 90 160.5639 -18.8037>}]
 upperarm_l [{
   shapeType=Capsule
   length=24.3802
   radius=6.25
   shapeName=upperarm_r_capsule
   center=<11.3453 0.1826 -1.3767>
   rotation=<rot -85 0.001 179.999>
  }]
 upperarm_r [{
   shapeType=Capsule
   length=24.3802
   radius=6.25
   shapeName=upperarm_r_capsule
   center=<-11.3453 -0.1826 1.3767>
   rotation=<rot -85 0.001 -0.001>
  }]
}
constraints [
 {bone1Name=cape_mid_01_r, bone2Name=cape_mid_r_MSTR, swing1LimitDegrees=20, swing2LimitDegrees=15, twistLimitDegrees=5}
 {bone1Name=cape_mid_02_r, bone2Name=cape_mid_01_r, swing1LimitDegrees=20, swing2LimitDegrees=15, twistLimitDegrees=5}
 {bone1Name=cape_02_r, bone2Name=cape_01_r, swing1LimitDegrees=30, swing2LimitDegrees=25, twistLimitDegrees=5}
 {
  bone1Name=spine_01
  bone2Name=pelvis
  swing1LimitDegrees=45
  swing1Motion=Free
  swing2LimitDegrees=45
  swing2Motion=Free
  twistLimitDegrees=45
  twistMotion=Free
 }
 {
  bone1Name=spine_02
  bone2Name=spine_01
  swing1LimitDegrees=45
  swing1Motion=Free
  swing2LimitDegrees=45
  swing2Motion=Free
  twistLimitDegrees=45
  twistMotion=Free
 }
 {
  bone1Name=upperarm_r
  bone2Name=spine_03
  swing1LimitDegrees=45
  swing1Motion=Free
  swing2LimitDegrees=45
  swing2Motion=Free
  twistLimitDegrees=45
  twistMotion=Free
 }
 {
  bone1Name=thigh_l
  bone2Name=pelvis
  swing1LimitDegrees=45
  swing1Motion=Free
  swing2LimitDegrees=45
  swing2Motion=Free
  twistLimitDegrees=45
  twistMotion=Free
 }
 {
  bone1Name=calf_r
  bone2Name=thigh_r
  swing1LimitDegrees=45
  swing1Motion=Free
  swing2LimitDegrees=45
  swing2Motion=Free
  twistLimitDegrees=45
  twistMotion=Free
 }
 {
  bone1Name=foot_r
  bone2Name=calf_r
  swing1LimitDegrees=45
  swing1Motion=Free
  swing2LimitDegrees=45
  swing2Motion=Free
  twistLimitDegrees=45
  twistMotion=Free
 }
 {
  bone1Name=lowerarm_r
  bone2Name=upperarm_r
  swing1LimitDegrees=45
  swing1Motion=Free
  swing2LimitDegrees=45
  swing2Motion=Free
  twistLimitDegrees=45
  twistMotion=Free
 }
 {bone1Name=cape_01_r, bone2Name=cape_r_MSTR, swing1LimitDegrees=30, swing2LimitDegrees=25, twistLimitDegrees=5}
 {
  bone1Name=thigh_r
  bone2Name=pelvis
  swing1LimitDegrees=45
  swing1Motion=Free
  swing2LimitDegrees=45
  swing2Motion=Free
  twistLimitDegrees=45
  twistMotion=Free
 }
 {
  bone1Name=upperarm_l
  bone2Name=spine_03
  swing1LimitDegrees=45
  swing1Motion=Free
  swing2LimitDegrees=45
  swing2Motion=Free
  twistLimitDegrees=45
  twistMotion=Free
 }
 {bone1Name=lowerarm_l, bone2Name=upperarm_l, swing1LimitDegrees=35, swing2LimitDegrees=35, twistLimitDegrees=15}
 {
  bone1Name=calf_l
  bone2Name=thigh_l
  swing1LimitDegrees=45
  swing1Motion=Free
  swing2LimitDegrees=45
  swing2Motion=Free
  twistLimitDegrees=45
  twistMotion=Free
 }
 {
  bone1Name=foot_l
  bone2Name=calf_l
  swing1LimitDegrees=45
  swing1Motion=Free
  swing2LimitDegrees=45
  swing2Motion=Free
  twistLimitDegrees=45
  twistMotion=Free
 }
 {
  bone1Name=hand_r
  bone2Name=lowerarm_r
  swing1LimitDegrees=45
  swing1Motion=Free
  swing2LimitDegrees=45
  swing2Motion=Free
  twistLimitDegrees=45
  twistMotion=Free
 }
 {
  bone1Name=cape_r_MSTR
  bone2Name=spine_03
  swing1LimitDegrees=45
  swing1Motion=Free
  swing2LimitDegrees=45
  swing2Motion=Free
  twistLimitDegrees=45
  twistMotion=Free
 }
 {
  bone1Name=cape_mid_r_MSTR
  bone2Name=spine_03
  swing1LimitDegrees=45
  swing1Motion=Free
  swing2LimitDegrees=45
  swing2Motion=Free
  twistLimitDegrees=45
  twistMotion=Free
 }
 {
  bone1Name=spine_03
  bone2Name=spine_02
  swing1LimitDegrees=45
  swing1Motion=Free
  swing2LimitDegrees=45
  swing2Motion=Free
  twistLimitDegrees=45
  twistMotion=Free
 }
 {bone1Name=cape_03_r, bone2Name=cape_02_r, swing1LimitDegrees=30, swing2LimitDegrees=25, twistLimitDegrees=5}
 {bone1Name=cape_mid_03_r, bone2Name=cape_mid_02_r, swing1LimitDegrees=20, swing2LimitDegrees=15, twistLimitDegrees=5}
 {bone1Name=cape_mid_01_l, bone2Name=cape_mid_l_MSTR, swing1LimitDegrees=20, swing2LimitDegrees=25, twistLimitDegrees=5}
 {
  bone1Name=cape_mid_l_MSTR
  bone2Name=spine_02
  swing1LimitDegrees=45
  swing1Motion=Free
  swing2LimitDegrees=45
  swing2Motion=Free
  twistLimitDegrees=45
  twistMotion=Free
 }
 {bone1Name=cape_mid_02_l, bone2Name=cape_mid_01_l, swing1LimitDegrees=20, swing2LimitDegrees=25, twistLimitDegrees=5}
]
