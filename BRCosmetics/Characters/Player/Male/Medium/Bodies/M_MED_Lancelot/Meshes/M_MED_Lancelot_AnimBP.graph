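# Post-process style Anim Blueprint for the Lancelot body part, driving cape, armor, bag and zipper secondary motion over the parent mesh's copied pose.
# Control Rig, constraints, AnimDynamics and additive corrective bones, weighted by measured limb angles, build cached base poses.
# Output blends backbling-specific RigidBody cloak sims, with cloth variants, or a procedural look-at fallback with a DBNO variant.
# EventGraph updates corrective alphas, gravity override and cloth wind settings.
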AnimBlueprint /BRCosmetics/Characters/Player/Male/Medium/Bodies/M_MED_Lancelot/Meshes/M_MED_Lancelot_AnimBP
target_skeleton=&/M_MED_Lancelot_Skeleton
graphs [
 AnimGraph {nodes [
   n1 = Get bIsDBNOCarried()
   n2 = "Transform (Modify) Bone - Bone: cape_mid_r_MSTR"!(
    ComponentPose=Local To Component(LocalPose=Use cached pose 'StartupCache'())
    bAlphaBoolEnabled:bool=@n1
    Translation="0,0,3"
    Rotation="60,0,0"
    Scale="1,1,1"
   ) -> (Pose) {
    alphaInputType=Bool
    rotationMode=BMM_Additive
    rotationSpace=BCS_BoneSpace
    scaleMode=BMM_Additive
    scaleSpace=BCS_BoneSpace
    translationMode=BMM_Additive
    translationSpace=BCS_BoneSpace
    alphaBoolBlend {blendOption=Linear}
    boneToModify {boneName=cape_mid_r_MSTR}
    rotation {pitch=60}
    translation {z=3}
   }
   n3 = "Transform (Modify) Bone - Bone: cape_r_MSTR"!(ComponentPose=@n2, bAlphaBoolEnabled:bool=@n1, Translation="0,0,0", Rotation="45,0,0", Scale="1,1,1") -> (Pose) {
    alphaInputType=Bool
    rotationMode=BMM_Additive
    rotationSpace=BCS_BoneSpace
    scaleMode=BMM_Additive
    scaleSpace=BCS_BoneSpace
    translationMode=BMM_Additive
    translationSpace=BCS_BoneSpace
    alphaBoolBlend {blendOption=Linear}
    boneToModify {boneName=cape_r_MSTR}
    rotation {pitch=45}
   }
   n4 = Get isWearBackbling() -> (isWearBackbling:bool)
   n5 = Get GravityOverride()
   n6 = RigidBody!(ComponentPose=@n3, bAlphaBoolEnabled:bool=NOT Boolean(A=@n4), OverrideWorldGravity:Vector=@n5, ExternalForce) -> (Pose) {
    alphaInputType=Bool
    bClampLinearTranslationLimitToRefPose=true
    bOverrideWorldGravity=true
    overridePhysicsAsset=&/M_MED_Lancelot_Dyn_Cloak
    simulationSpace=BaseBoneSpace
    alphaBoolBlend {blendOption=Linear}
    baseBoneRef {boneName=spine_01}
    componentAppliedLinearAccClamp=<3000 3000 3000>
    componentLinearAccScale {x=0.2}
    componentLinearVelScale {x=0.2}
    overrideWorldGravity {z=-980}
    simSpaceSettings {
     maxAngularAcceleration=200
     maxAngularVelocity=8
     maxLinearAcceleration=2000
     maxLinearVelocity=500
     velocityScaleZ=0.5
     worldAlpha=1
     externalLinearDragV=<1 1>
    }
   }
   n7 = RigidBody(ComponentPose=@n6, bAlphaBoolEnabled=@n4, OverrideWorldGravity=@n5) {
    alphaInputType=Bool
    bClampLinearTranslationLimitToRefPose=true
    bOverrideWorldGravity=true
    overridePhysicsAsset=&/M_MED_Lancelot_Dyn_Cloak_Wearing_Backpack
    simulationSpace=BaseBoneSpace
    alphaBoolBlend {blendOption=Linear}
    baseBoneRef {boneName=spine_01}
    componentAppliedLinearAccClamp=<3000 3000 3000>
    componentLinearAccScale {x=0.2}
    componentLinearVelScale {x=0.2}
    overrideWorldGravity {z=-980}
    simSpaceSettings {
     maxAngularAcceleration=200
     maxAngularVelocity=8
     maxLinearAcceleration=2000
     maxLinearVelocity=500
     velocityScaleZ=0.5
     worldAlpha=1
     externalLinearDragV=<1 1>
    }
   }
   n8 = Component To Local(ComponentPose=@n7)
   n9 = "Transform (Modify) Bone - Bone: cape_r_MSTR"(
    ComponentPose=Local To Component(LocalPose=Use cached pose 'StartupCache'())
    bAlphaBoolEnabled=True
    Translation="0,0,0"
    Rotation="30,-30,0"
    Scale="1,1,1"
   ) {
    alphaInputType=Bool
    rotationMode=BMM_Additive
    rotationSpace=BCS_BoneSpace
    scaleMode=BMM_Additive
    scaleSpace=BCS_BoneSpace
    translationMode=BMM_Additive
    translationSpace=BCS_BoneSpace
    alphaBoolBlend {blendOption=Linear}
    boneToModify {boneName=cape_r_MSTR}
    rotation {pitch=30, yaw=-30}
   }
   n10 = Get isWearBackbling()
   n11 = RigidBody(ComponentPose=@n9, bAlphaBoolEnabled=NOT Boolean(A=@n10), OverrideWorldGravity=@n5) {
    alphaInputType=Bool
    bClampLinearTranslationLimitToRefPose=true
    bOverrideWorldGravity=true
    overridePhysicsAsset=&/M_MED_Lancelot_Dyn_Cloak_For_Cloth
    simulationSpace=BaseBoneSpace
    alphaBoolBlend {blendOption=Linear}
    baseBoneRef {boneName=spine_01}
    componentAppliedLinearAccClamp=<3000 3000 3000>
    componentLinearAccScale {x=0.2}
    componentLinearVelScale {x=0.2}
    overrideWorldGravity {z=-980}
    simSpaceSettings {
     maxAngularAcceleration=200
     maxAngularVelocity=8
     maxLinearAcceleration=2000
     maxLinearVelocity=500
     velocityScaleZ=0.5
     worldAlpha=1
     externalLinearDragV=<1 1>
    }
   }
   n12 = RigidBody(ComponentPose=@n11, bAlphaBoolEnabled=@n10, OverrideWorldGravity=@n5) {
    alphaInputType=Bool
    bClampLinearTranslationLimitToRefPose=true
    bOverrideWorldGravity=true
    overridePhysicsAsset=&/M_MED_Lancelot_Dyn_Cloak_Wearing_Pack_For_Cloth
    simulationSpace=BaseBoneSpace
    alphaBoolBlend {blendOption=Linear}
    baseBoneRef {boneName=spine_01}
    componentAppliedLinearAccClamp=<3000 3000 3000>
    componentLinearAccScale {x=0.2}
    componentLinearVelScale {x=0.2}
    overrideWorldGravity {z=-980}
    simSpaceSettings {
     maxAngularAcceleration=200
     maxAngularVelocity=8
     maxLinearAcceleration=2000
     maxLinearVelocity=500
     velocityScaleZ=0.5
     worldAlpha=1
     externalLinearDragV=<1 1>
    }
   }
   n13 = Component To Local(ComponentPose=@n12)
   n14 = Get bClothEnabled()
   n15 = Get arm_rot_bk_r() -> (arm_rot_bk_r:double)
   n16 = "Transform (Modify) Bone - Bone: cape_01_r"(
    ComponentPose=Local To Component(LocalPose=Use cached pose 'NoRigidBasePose'())
    Alpha=@n15
    Translation="0,0,3"
    Rotation="30,0.000,0.000"
    Scale="1,1,1"
   ) {
    rotationMode=BMM_Additive
    rotationSpace=BCS_BoneSpace
    scaleMode=BMM_Additive
    scaleSpace=BCS_BoneSpace
    translationMode=BMM_Additive
    translationSpace=BCS_BoneSpace
    alphaBoolBlend {blendOption=Linear}
    boneToModify {boneName=cape_01_r}
    rotation {pitch=30}
    translation {z=3}
   }
   n17 = "Transform (Modify) Bone - Bone: cape_02_r"(ComponentPose=@n16, Alpha=@n15, Translation="-0.000,0.000,4", Rotation="15,0,0", Scale="1,1,1") {
    rotationMode=BMM_Additive
    rotationSpace=BCS_BoneSpace
    scaleMode=BMM_Additive
    scaleSpace=BCS_BoneSpace
    translationMode=BMM_Additive
    translationSpace=BCS_BoneSpace
    alphaBoolBlend {blendOption=Linear}
    boneToModify {boneName=cape_02_r}
    rotation {pitch=15}
    translation {z=4}
   }
   n18 = "Transform (Modify) Bone - Bone: cape_03_r"(ComponentPose=@n17, Alpha=@n15, Translation="-0.000,0.000,2", Rotation="0,0,0", Scale="1,1,1") {
    rotationMode=BMM_Additive
    rotationSpace=BCS_BoneSpace
    scaleMode=BMM_Additive
    scaleSpace=BCS_BoneSpace
    translationMode=BMM_Additive
    translationSpace=BCS_BoneSpace
    alphaBoolBlend {blendOption=Linear}
    boneToModify {boneName=cape_03_r}
    translation {z=2}
   }
   n19 = Look At - Bone: cape_mid_03_l(ComponentPose=@n18, Alpha=Get cloak_aim_alpha_l() -> (cloak_aim_alpha_l:double)) {
    alphaBoolBlend {blendOption=Linear}
    boneToModify {boneName=cape_mid_03_l}
    lookAtLocation=<-43.3685 -10.6666 15.1702>
    lookAtTarget {boneReference {boneName=calf_twist_01_l}}
    lookAt_Axis {axis=<1 0>}
   }
   n20 = Get cloak_aim_alpha_r() -> (cloak_aim_alpha_r:double)
   n21 = Look At - Bone: cape_mid_04_r(ComponentPose=@n19, Alpha=@n20) {
    alphaBoolBlend {blendOption=Linear}
    boneToModify {boneName=cape_mid_04_r}
    lookAtLocation=<80.1232 6.9709 -5.5159>
    lookAtTarget {boneReference {boneName=calf_twist_01_r}}
    lookAt_Axis {axis=<1 0>}
   }
   n22 = Look At - Bone: cape_04_r(ComponentPose=@n21, Alpha=@n20) {
    alphaBoolBlend {blendOption=Linear}
    boneToModify {boneName=cape_04_r}
    lookAtLocation=<80.1232 5.3369 11.5162>
    lookAtTarget {boneReference {boneName=calf_twist_01_r}}
    lookAt_Axis {axis=<1 0>}
   }
   n23 = "Transform (Modify) Bone - Bone: cape_04_r"(ComponentPose=@n22, Alpha=Get arm_rot_bk_r(), Translation="0,0,0", Rotation="15,0,0", Scale="1,1,1") {
    rotationMode=BMM_Additive
    rotationSpace=BCS_BoneSpace
    scaleMode=BMM_Additive
    scaleSpace=BCS_BoneSpace
    translationMode=BMM_Additive
    translationSpace=BCS_BoneSpace
    alphaBoolBlend {blendOption=Linear}
    boneToModify {boneName=cape_04_r}
    rotation {pitch=15}
   }
   n24 = Component To Local(ComponentPose=@n23)
   n25 = "Transform (Modify) Bone - Bone: cape_mid_03_r"(
    ComponentPose=Local To Component(LocalPose=Use cached pose 'NoRigidBasePose'())
    Alpha=1
    Translation="0,0,0"
    Rotation="-12,0,0"
    Scale="1,1,1"
   ) {
    rotationMode=BMM_Additive
    rotationSpace=BCS_BoneSpace
    scaleMode=BMM_Additive
    scaleSpace=BCS_BoneSpace
    translationMode=BMM_Additive
    translationSpace=BCS_BoneSpace
    alphaBoolBlend {blendOption=Linear}
    boneToModify {boneName=cape_mid_03_r}
    rotation {pitch=-12}
   }
   n26 = "Transform (Modify) Bone - Bone: cape_03_r"(ComponentPose=@n25, Alpha=1, Translation="0,0,0", Rotation="-15,0,0", Scale="1,1,1") {
    rotationMode=BMM_Additive
    rotationSpace=BCS_BoneSpace
    scaleMode=BMM_Additive
    scaleSpace=BCS_BoneSpace
    translationMode=BMM_Additive
    translationSpace=BCS_BoneSpace
    alphaBoolBlend {blendOption=Linear}
    boneToModify {boneName=cape_03_r}
    rotation {pitch=-15}
   }
   n27 = "Transform (Modify) Bone - Bone: cape_02_r"!(ComponentPose=@n26, bAlphaBoolEnabled:bool=Get bIsDBNOCarried(), Translation="0,0,-5", Rotation="-15,0,0", Scale="1,1,1") -> (Pose) {
    alphaInputType=Bool
    rotationMode=BMM_Additive
    rotationSpace=BCS_BoneSpace
    scaleMode=BMM_Additive
    scaleSpace=BCS_BoneSpace
    translationMode=BMM_Additive
    translationSpace=BCS_BoneSpace
    alphaBoolBlend {blendOption=Linear}
    boneToModify {boneName=cape_02_r}
    rotation {pitch=-15}
    translation {z=-5}
   }
   n28 = "Transform (Modify) Bone - Bone: cape_mid_02_l"(ComponentPose=@n27, Alpha=1, Translation="0,0,0", Rotation="-12,0,0", Scale="1,1,1") {
    rotationMode=BMM_Additive
    rotationSpace=BCS_BoneSpace
    scaleMode=BMM_Additive
    scaleSpace=BCS_BoneSpace
    translationMode=BMM_Additive
    translationSpace=BCS_BoneSpace
    alphaBoolBlend {blendOption=Linear}
    boneToModify {boneName=cape_mid_02_l}
    rotation {pitch=-12}
   }
   n29 = Get bIsDBNOCarried()
   n30 = Look At - Bone: cape_04_r!(ComponentPose=@n28, bAlphaBoolEnabled:bool=NOT Boolean(A=@n29)) -> (Pose) {
    alphaInputType=Bool
    alphaBoolBlend {blendOption=Linear}
    boneToModify {boneName=cape_04_r}
    lookAtLocation=<17.4555 -6.3536 12.7776>
    lookAtTarget {boneReference {boneName=foot_r}}
    lookAt_Axis {axis=<1 0>}
   }
   n31 = Look At - Bone: cape_04_r(ComponentPose=@n30, bAlphaBoolEnabled=@n29) {
    alphaInputType=Bool
    alphaBoolBlend {blendOption=Linear}
    boneToModify {boneName=cape_04_r}
    lookAtLocation=<-19.8775 -42.0456 29.0528>
    lookAtTarget {boneReference {boneName=foot_r}}
    lookAt_Axis {axis=<1 0>}
   }
   n32 = "Constraint - cape_mid_04_r (PP)"!(ComponentPose=@n31, Alpha=1, ConstraintWeights_0=0.65, ConstraintWeights_1:float=0.35) -> (Pose) {
    alphaBoolBlend {blendOption=Linear}
    boneToModify {boneName=cape_mid_04_r}
    constraintSetup [{targetBone {boneName=cape_04_r}}, {targetBone {boneName=cape_mid_03_l}}]
    constraintWeights [1,1]
   }
   n33 = Component To Local(ComponentPose=@n32)
   n34 = Get bIsDBNO()
   Output Pose(Result=Blend Poses by bool(
     BlendPose_0=Blend Poses by bool(BlendPose_0=@n13, BlendPose_1=@n8, BlendTime_0=0.2, BlendTime_1=0.5, bActiveValue=@n14) {blendType=Linear}
     BlendPose_1=Blend Poses by bool(BlendPose_0=@n33, BlendPose_1=@n24, BlendTime_0=0.5, BlendTime_1=0.5, bActiveValue=@n34) {blendType=Linear}
     BlendTime_0=0.2
     BlendTime_1=0.5
     bActiveValue=Get bRigidBodyEnabled()
    ) {blendType=Linear}) {layerGroup=None}
   n35 = Control Rig!(
    Source=Copy Pose From Mesh() {bUseAttachedParent=true}
    Alpha
    KneeRot_fr_l_114deg:Vector
    KneeTrans_fr_l_114deg:Vector="0.258,-3.591,0.009"
    KneeScale_fr_l_114deg:Vector="1,1,1"
    KneeScale_fr_l_130deg:Vector="1,1,1"
    KneeRot_fr_l_130deg:Vector="0,0,-51.2"
    KneeTrans_fr_l_130deg:Vector="0.258,-3.591,0.009"
   ) -> (Pose) {
    alphaBoolBlend {blendOption=Linear}
    controlRigAssetReference {blueprintRigClass=&/M_MED_Lancelot_Deform_CtrlRig_sculpters.M_MED_Lancelot_Deform_CtrlRig_sculpters_C}
   }
   n36 = "Constraint - dyn_zipper_MSTR (PPPPP)"!(
    ComponentPose=Local To Component(LocalPose=@n35)
    Alpha=1
    ConstraintWeights_0=0.2174
    ConstraintWeights_1:float=0.0757
    ConstraintWeights_2:float=0.7038
    ConstraintWeights_3:float=0.0009
    ConstraintWeights_4:float=0.0022
   ) -> (Pose) {
    alphaBoolBlend {blendOption=Linear}
    boneToModify {boneName=dyn_zipper_MSTR}
    constraintSetup [
     {targetBone {boneName=thigh_l}}
     {targetBone {boneName=pelvis}}
     {targetBone {boneName=thigh_twist_01_l}}
     {targetBone {boneName=thigh_twist_01_r}}
     {targetBone {boneName=calf_l}}
    ]
    constraintWeights [1,1,1,1,1]
   }
   n37 = "Constraint - dyn_zipper_TGT (PPPPP)"!(
    ComponentPose=@n36
    Alpha=1
    ConstraintWeights_0=0.8558
    ConstraintWeights_1:float=0.0216
    ConstraintWeights_2:float=0.0007
    ConstraintWeights_3:float=0.0207
    ConstraintWeights_4:float=0.1012
   ) -> (Pose) {
    alphaBoolBlend {blendOption=Linear}
    boneToModify {boneName=dyn_zipper_TGT}
    constraintSetup [
     {targetBone {boneName=thigh_l}}
     {targetBone {boneName=thigh_twist_01_l}}
     {targetBone {boneName=calf_twist_02_l}}
     {targetBone {boneName=calf_l}}
     {targetBone {boneName=deform_knee_fr_01_l}}
    ]
    constraintWeights [1,1,1,1,1]
   }
   n38 = Look At - Bone: dyn_zipper_MSTR(ComponentPose=@n37, Alpha=1) {
    alphaBoolBlend {blendOption=Linear}
    boneToModify {boneName=dyn_zipper_MSTR}
    lookAtLocation {x=10}
    lookAtTarget {boneReference {boneName=dyn_zipper_TGT}}
    lookAt_Axis {axis=<1 0>}
   }
   n39 = Anim Dynamics - Bone: dyn_zipper!(ComponentPose=@n38, Alpha=1, GravityScale:float=1, GravityOverride:Vector="0,0,-980", ExternalForce) -> (Pose) {
    angularDampingOverride=0.99
    angularSpringConstant=1000
    bAngularSpring=true
    bGravityOverrideInSimSpace=true
    bOverrideAngularDamping=true
    bUseGravityOverride=true
    simulationSpace=RootRelative
    alphaBoolBlend {blendOption=Linear}
    boundBone {boneName=dyn_zipper}
    componentAppliedLinearAccClamp=<5000 5000 5000>
    componentLinearAccScale=<0.6 0.6 0.6>
    componentLinearVelScale=<0.6 0.6 0.6>
    gravityOverride {z=-980}
    physicsBodyDefinitions [{
      sphereCollisionRadius=0
      boundBone {boneName=dyn_zipper}
      boxExtents=<12 10 10>
      constraintSetup {angularLimitsMax=<0 0 35>, angularLimitsMin=<0 -45 -35>, angularTarget=<0 0 0>}
      localJointOffset=<1.5 0 0>
     }]
    relativeSpaceBone {boneName=head}
   }
   n40 = "Constraint - def_bag_01_MSTR (PPPPP)"!(
    ComponentPose=@n39
    Alpha=1
    ConstraintWeights_0=0.155
    ConstraintWeights_1:float=0.0414
    ConstraintWeights_2:float=0.6882
    ConstraintWeights_3:float=0.0357
    ConstraintWeights_4:float=0.0798
   ) -> (Pose) {
    alphaBoolBlend {blendOption=Linear}
    boneToModify {boneName=def_bag_01_MSTR}
    constraintSetup [
     {targetBone {boneName=spine_01}}
     {targetBone {boneName=spine_02}}
     {targetBone {boneName=pelvis}}
     {targetBone {boneName=spine_03}}
     {targetBone {boneName=thigh_twist_01_r}}
    ]
    constraintWeights [1,1,1,1,1]
   }
   n41 = "Constraint - def_bag_02_MSTR (PPPPPP)"!(
    ComponentPose=@n40
    Alpha=1
    ConstraintWeights_0=0.1837
    ConstraintWeights_1:float=0.0639
    ConstraintWeights_2:float=0.6569
    ConstraintWeights_3:float=0
    ConstraintWeights_4:float=0.0553
    ConstraintWeights_5:float=0.0402
   ) -> (Pose) {
    alphaBoolBlend {blendOption=Linear}
    boneToModify {boneName=def_bag_02_MSTR}
    constraintSetup [
     {targetBone {boneName=spine_01}}
     {targetBone {boneName=spine_02}}
     {targetBone {boneName=pelvis}}
     {targetBone {boneName=thigh_twist_01_l}}
     {targetBone {boneName=spine_03}}
     {targetBone {boneName=thigh_twist_01_r}}
    ]
    constraintWeights [1,1,1,1,1,1]
   }
   n42 = Get thigh_rot_up_r() -> (thigh_rot_up_r:double)
   n43 = "Transform (Modify) Bone - Bone: def_bag_01"(ComponentPose=@n41, Alpha=@n42, Translation="0,0,0", Rotation="120,0,0", Scale="1,1,1") {
    rotationMode=BMM_Additive
    rotationSpace=BCS_BoneSpace
    translationMode=BMM_Additive
    translationSpace=BCS_BoneSpace
    alphaBoolBlend {blendOption=Linear}
    boneToModify {boneName=def_bag_01}
    rotation {pitch=120}
   }
   n44 = "Transform (Modify) Bone - Bone: def_bag_02"(ComponentPose=@n43, Alpha=@n42, Translation="0,0,0", Rotation="120,0,0", Scale="1,1,1") {
    rotationMode=BMM_Additive
    rotationSpace=BCS_BoneSpace
    translationMode=BMM_Additive
    translationSpace=BCS_BoneSpace
    alphaBoolBlend {blendOption=Linear}
    boneToModify {boneName=def_bag_02}
    rotation {pitch=120}
   }
   n45 = "Transform (Modify) Bone - Bone: attach_backpack"(ComponentPose=@n44, Alpha=Get chest_twist() -> (chest_twist:double), Translation="0,-2,0", Rotation="0,0,-15", Scale="1,1,1") {
    rotationMode=BMM_Additive
    rotationSpace=BCS_BoneSpace
    scaleMode=BMM_Additive
    scaleSpace=BCS_BoneSpace
    translationMode=BMM_Additive
    translationSpace=BCS_BoneSpace
    alphaBoolBlend {blendOption=Linear}
    boneToModify {boneName=attach_backpack}
    rotation {roll=-15}
    translation {y=-2}
   }
   n46 = Get arm_rot_dn_r() -> (arm_rot_dn_r:double)
   n47 = "Transform (Modify) Bone - Bone: def_cloak_01"(ComponentPose=@n45, Alpha=@n46, Translation="0,0,0", Rotation="20,15,0", Scale="1,1,1") {
    rotationMode=BMM_Additive
    rotationSpace=BCS_BoneSpace
    scaleMode=BMM_Additive
    scaleSpace=BCS_BoneSpace
    translationMode=BMM_Additive
    translationSpace=BCS_BoneSpace
    alphaBoolBlend {blendOption=Linear}
    boneToModify {boneName=def_cloak_01}
    rotation {pitch=20, yaw=15}
   }
   n48 = "Transform (Modify) Bone - Bone: def_cloak_02"(ComponentPose=@n47, Alpha=@n46, Translation="0,0,0", Rotation="5,0,0", Scale="1,1,1") {
    rotationMode=BMM_Additive
    rotationSpace=BCS_BoneSpace
    scaleMode=BMM_Additive
    scaleSpace=BCS_BoneSpace
    translationMode=BMM_Additive
    translationSpace=BCS_BoneSpace
    alphaBoolBlend {blendOption=Linear}
    boneToModify {boneName=def_cloak_02}
    rotation {pitch=5}
   }
   n49 = "Transform (Modify) Bone - Bone: cape_mid_r_MSTR"!(ComponentPose=@n48, Alpha:float=@n46, Translation="0.000,5,0", Rotation="0,0,0", Scale="1,1,1") -> (Pose) {
    rotationMode=BMM_Additive
    rotationSpace=BCS_BoneSpace
    scaleMode=BMM_Additive
    scaleSpace=BCS_BoneSpace
    translationMode=BMM_Additive
    translationSpace=BCS_BoneSpace
    alphaBoolBlend {blendOption=Linear}
    boneToModify {boneName=cape_mid_r_MSTR}
    translation {y=5}
   }
   n50 = "Transform (Modify) Bone - Bone: cape_r_MSTR"!(
    ComponentPose=@n49
    Alpha:float=Get clavicle_rot_up_r() -> (clavicle_rot_up_r:double)
    Translation="11,18,0.000"
    Rotation="-0.000,50.000,-0.000"
    Scale="1,1,1"
   ) -> (Pose) {
    rotationMode=BMM_Additive
    rotationSpace=BCS_BoneSpace
    scaleMode=BMM_Additive
    scaleSpace=BCS_BoneSpace
    translationMode=BMM_Additive
    translationSpace=BCS_BoneSpace
    alphaBoolBlend {blendOption=Linear}
    boneToModify {boneName=cape_r_MSTR}
    rotation=<rot 0 50 0>
    translation=<11 18>
   }
   n51 = Get arm_rot_bk_r()
   n52 = "Transform (Modify) Bone - Bone: cape_r_MSTR"(ComponentPose=@n50, Alpha=@n51, Translation="0.000,12,18", Rotation="0,0,0", Scale="1,1,1") {
    rotationMode=BMM_Additive
    rotationSpace=BCS_BoneSpace
    scaleMode=BMM_Additive
    scaleSpace=BCS_BoneSpace
    translationMode=BMM_Additive
    translationSpace=BCS_BoneSpace
    alphaBoolBlend {blendOption=Linear}
    boneToModify {boneName=cape_r_MSTR}
    translation {y=12, z=18}
   }
   n53 = "Transform (Modify) Bone - Bone: cape_mid_r_MSTR"(ComponentPose=@n52, Alpha=@n51, Translation="0,0,0", Rotation="0,0,0", Scale="1,1,1") {
    rotationMode=BMM_Additive
    rotationSpace=BCS_BoneSpace
    scaleMode=BMM_Additive
    scaleSpace=BCS_BoneSpace
    translationMode=BMM_Additive
    translationSpace=BCS_BoneSpace
    alphaBoolBlend {blendOption=Linear}
    boneToModify {boneName=cape_mid_r_MSTR}
   }
   n54 = Get def_cape_parachute() -> (def_cape_parachute:double)
   n55 = "Transform (Modify) Bone - Bone: def_cloak_01"(ComponentPose=@n53, Alpha=@n54, Translation="-5,0,0", Rotation="0,0,0", Scale="1,1,1") {
    rotationMode=BMM_Additive
    rotationSpace=BCS_BoneSpace
    scaleMode=BMM_Additive
    scaleSpace=BCS_BoneSpace
    translationMode=BMM_Additive
    translationSpace=BCS_BoneSpace
    alphaBoolBlend {blendOption=Linear}
    boneToModify {boneName=def_cloak_01}
    translation {x=-5}
   }
   n56 = "Transform (Modify) Bone - Bone: def_cloak_03"(ComponentPose=@n55, Alpha=@n54, Translation="0,0,0", Rotation="-50,0,0", Scale="1,1,1") {
    rotationMode=BMM_Additive
    rotationSpace=BCS_BoneSpace
    scaleMode=BMM_Additive
    scaleSpace=BCS_BoneSpace
    translationMode=BMM_Additive
    translationSpace=BCS_BoneSpace
    alphaBoolBlend {blendOption=Linear}
    boneToModify {boneName=def_cloak_03}
    rotation {pitch=-50}
   }
   n57 = Get arm_rot_up_r() -> (arm_rot_up_r:double)
   n58 = "Transform (Modify) Bone - Bone: def_cloak_01"(ComponentPose=@n56, Alpha=@n57, Translation="-15,0,-0.000", Rotation="0,0,0", Scale="1,1,1") {
    rotationMode=BMM_Additive
    rotationSpace=BCS_BoneSpace
    scaleMode=BMM_Additive
    scaleSpace=BCS_BoneSpace
    translationMode=BMM_Additive
    translationSpace=BCS_BoneSpace
    alphaBoolBlend {blendOption=Linear}
    boneToModify {boneName=def_cloak_01}
    translation {x=-15}
   }
   n59 = "Transform (Modify) Bone - Bone: def_cloak_03"(ComponentPose=@n58, Alpha=@n57, Translation="-4.077,-11.516,7.125", Rotation="0,90,0", Scale="1,1,1") {
    rotationMode=BMM_Additive
    rotationSpace=BCS_BoneSpace
    scaleMode=BMM_Additive
    scaleSpace=BCS_BoneSpace
    translationMode=BMM_Additive
    translationSpace=BCS_BoneSpace
    alphaBoolBlend {blendOption=Linear}
    boneToModify {boneName=def_cloak_03}
    rotation {yaw=90}
    translation=<-4.077 -11.516 7.125>
   }
   n60 = "Transform (Modify) Bone - Bone: cape_r_MSTR"(ComponentPose=@n59, Alpha=@n57, Translation="2.000,6.473,-0.000", Rotation="0,0,0", Scale="1,1,1") {
    rotationMode=BMM_Additive
    rotationSpace=BCS_BoneSpace
    scaleMode=BMM_Additive
    scaleSpace=BCS_BoneSpace
    translationMode=BMM_Additive
    translationSpace=BCS_BoneSpace
    alphaBoolBlend {blendOption=Linear}
    boneToModify {boneName=cape_r_MSTR}
    translation=<2 6.473>
   }
   n61 = "Transform (Modify) Bone - Bone: def_armor_arm_l_MSTR"(ComponentPose=@n60, Alpha=Get hand_rot_l() -> (hand_rot_l:double), Translation="-1.309,-0.000,0.549", Rotation="0,0,0", Scale="1,1,1") {
    rotationMode=BMM_Additive
    rotationSpace=BCS_BoneSpace
    scaleMode=BMM_Additive
    scaleSpace=BCS_BoneSpace
    translationMode=BMM_Additive
    translationSpace=BCS_BoneSpace
    alphaBoolBlend {blendOption=Linear}
    boneToModify {boneName=def_armor_arm_l_MSTR}
    translation {x=-1.309, z=0.549}
   }
   n62 = "Transform (Modify) Bone - Bone: def_armor_arm_l"(ComponentPose=@n61, Alpha=Get hand_twist_l() -> (hand_twist_l:double), Translation="0,0,0", Rotation="10.010,-1.871,0", Scale="1,1,1") {
    rotationMode=BMM_Additive
    rotationSpace=BCS_BoneSpace
    scaleMode=BMM_Additive
    scaleSpace=BCS_BoneSpace
    translationMode=BMM_Additive
    translationSpace=BCS_BoneSpace
    alphaBoolBlend {blendOption=Linear}
    boneToModify {boneName=def_armor_arm_l}
    rotation {pitch=10.01, yaw=-1.871}
   }
   n63 = Get isIdleEmotePlaying() -> (isIdleEmotePlaying:bool)
   n64 = "Transform (Modify) Bone - Bone: def_cloak_01"!(ComponentPose=@n62, bAlphaBoolEnabled:bool=@n63, Translation="-6.5,-2.1,1", Rotation="2.969,0.061,-32.470", Scale="1,1,1") -> (Pose) {
    alphaInputType=Bool
    rotationMode=BMM_Additive
    rotationSpace=BCS_BoneSpace
    scaleMode=BMM_Additive
    scaleSpace=BCS_BoneSpace
    translationMode=BMM_Additive
    translationSpace=BCS_BoneSpace
    alphaBoolBlend {blendOption=Linear}
    boneToModify {boneName=def_cloak_01}
    rotation=<rot 2.969 0.061 -32.47>
    translation=<-6.5 -2.1 1>
   }
   n65 = "Transform (Modify) Bone - Bone: cape_r_MSTR"!(ComponentPose=@n64, bAlphaBoolEnabled:bool=@n63, Translation="0,-3,0", Rotation="0,0,0", Scale="1,1,1") -> (Pose) {
    alphaInputType=Bool
    rotationMode=BMM_Additive
    rotationSpace=BCS_BoneSpace
    scaleMode=BMM_Additive
    scaleSpace=BCS_BoneSpace
    translationMode=BMM_Additive
    translationSpace=BCS_BoneSpace
    alphaBoolBlend {blendOption=Linear}
    boneToModify {boneName=cape_r_MSTR}
    translation {y=-3}
   }
   n66 = "Transform (Modify) Bone - Bone: deform_shoulder_armor_ft"!(ComponentPose=@n65, bAlphaBoolEnabled:bool=@n63, Translation="10,0,12", Rotation="-10,0,-20", Scale="1,1,1") -> (Pose) {
    alphaInputType=Bool
    rotationMode=BMM_Additive
    rotationSpace=BCS_BoneSpace
    scaleMode=BMM_Additive
    scaleSpace=BCS_BoneSpace
    translationMode=BMM_Additive
    translationSpace=BCS_BoneSpace
    alphaBoolBlend {blendOption=Linear}
    boneToModify {boneName=deform_shoulder_armor_ft}
    rotation {pitch=-10, roll=-20}
    translation {x=10, z=12}
   }
   n67 = "Transform (Modify) Bone - Bone: cape_r_MSTR"(ComponentPose=@n66, bAlphaBoolEnabled=True, Translation="2,2,0", Rotation="0,20,0", Scale="1,1,1") {
    alphaInputType=Bool
    rotationMode=BMM_Additive
    rotationSpace=BCS_BoneSpace
    scaleMode=BMM_Additive
    scaleSpace=BCS_BoneSpace
    translationMode=BMM_Additive
    translationSpace=BCS_BoneSpace
    alphaBoolBlend {blendOption=Linear}
    boneToModify {boneName=cape_r_MSTR}
    rotation {yaw=20}
    translation=<2 2>
   }
   n68 = "Transform (Modify) Bone - Bone: def_cloak_01"(ComponentPose=@n67, bAlphaBoolEnabled=True, Translation="0,0,0", Rotation="-8.124,-4.680,-2.672", Scale="1,1,1") {
    alphaInputType=Bool
    rotationMode=BMM_Additive
    rotationSpace=BCS_BoneSpace
    scaleMode=BMM_Additive
    scaleSpace=BCS_BoneSpace
    translationMode=BMM_Additive
    translationSpace=BCS_BoneSpace
    alphaBoolBlend {blendOption=Linear}
    boneToModify {boneName=def_cloak_01}
    rotation=<rot -8.124 -4.68 -2.672>
   }
   Save cached pose 'StartupCache'(Pose=Component To Local(ComponentPose=@n68)) {cacheName=StartupCache}
   n69 = Look At - Bone: cape_mid_01_l(ComponentPose=Local To Component(LocalPose=Use cached pose 'StartupCache'()), Alpha=1) {
    alphaBoolBlend {blendOption=Linear}
    boneToModify {boneName=cape_mid_01_l}
    lookAtLocation=<-38.2263 -16.1102 -5.1847>
    lookAtTarget {boneReference {boneName=pelvis}}
    lookAt_Axis {axis=<1 0>}
   }
   n70 = "Constraint - cape_mid_02_l (P)"(ComponentPose=@n69, Alpha=1, ConstraintWeights_0=1) {
    alphaBoolBlend {blendOption=Linear}
    boneToModify {boneName=cape_mid_02_l}
    constraintSetup [{targetBone {boneName=pelvis}}]
    constraintWeights [1]
   }
   n71 = "Constraint - cape_mid_01_r (PP)"!(ComponentPose=@n70, Alpha=1, ConstraintWeights_0=0.5, ConstraintWeights_1:float=0.5) -> (Pose) {
    alphaBoolBlend {blendOption=Linear}
    boneToModify {boneName=cape_mid_01_r}
    constraintSetup [{targetBone {boneName=spine_04}}, {targetBone {boneName=spine_02}}]
    constraintWeights [1,1]
   }
   n72 = "Constraint - cape_mid_02_r (PP)"!(ComponentPose=@n71, Alpha=1, ConstraintWeights_0=0.3, ConstraintWeights_1:float=0.7) -> (Pose) {
    alphaBoolBlend {blendOption=Linear}
    boneToModify {boneName=cape_mid_02_r}
    constraintSetup [{targetBone {boneName=spine_02}}, {targetBone {boneName=spine_01}}]
    constraintWeights [1,1]
   }
   n73 = "Constraint - cape_mid_03_r (P)"(ComponentPose=@n72, Alpha=1, ConstraintWeights_0=1) {
    alphaBoolBlend {blendOption=Linear}
    boneToModify {boneName=cape_mid_03_r}
    constraintSetup [{targetBone {boneName=pelvis}}]
    constraintWeights [1]
   }
   n74 = Look At - Bone: cape_r_MSTR(ComponentPose=@n73, Alpha=1) {
    alphaBoolBlend {blendOption=Linear}
    boneToModify {boneName=cape_r_MSTR}
    lookAtLocation=<-23.6047 -11.0754 29.2516>
    lookAtTarget {boneReference {boneName=pelvis}}
    lookAt_Axis {axis=<1 0>}
   }
   n75 = "Constraint - cape_02_r (PP)"!(ComponentPose=@n74, Alpha=0, ConstraintWeights_0=0.5, ConstraintWeights_1:float=0.5) -> (Pose) {
    alphaBoolBlend {blendOption=Linear}
    boneToModify {boneName=cape_02_r}
    constraintSetup [{targetBone {boneName=cape_01_r}}, {targetBone {boneName=cape_03_r}}]
    constraintWeights [1,1]
   }
   n76 = "Constraint - cape_03_r (P)"(ComponentPose=@n75, Alpha=0, ConstraintWeights_0=1) {
    alphaBoolBlend {blendOption=Linear}
    boneToModify {boneName=cape_03_r}
    constraintSetup [{targetBone {boneName=pelvis}}]
    constraintWeights [1]
   }
   Save cached pose 'NoRigidBasePose'(Pose=Component To Local(ComponentPose=@n76)) {cacheName=NoRigidBasePose}
  ]}
 EventGraph {nodes [
   Event BlueprintInitializeAnimation()
   n1 = TryGetFortPawn()
   > Is Valid(InputObject=@n1)
   > n2 = Cast To FortPlayerPawnAthena(Object=@n1)
   > Is Valid(InputObject=@n2)
   > n3 = Set Player_Pawn(Player_Pawn:Fort Player Pawn Object Reference=@n2) -> (Output_Get:Fort Player Pawn Object Reference)
   n4 = GetAnimInstance(self=Get Mesh(self=@n3))
   > Is Valid(InputObject=@n4)
   > Set Anim_Inst(Anim_Inst:Anim Instance Object Reference=@n4) -> (Output_Get:Anim Instance Object Reference)
   Event OnCharacterCustomizationLoaded()
   > Do Once()
   n5 = CheckForBackbling()
   > Set isWearBackbling(isWearBackbling:bool=@n5) -> (Output_Get:bool)
   > Set isWearCape(isWearCape:bool=@n5.bIsBackblingLong) -> (Output_Get:bool)
   Event BlueprintUpdateAnimation()
   > n6 = Sequence!(execute) -> (then_0, then_1, then_2:Exec, then_3:Exec, then_4:Exec)
   n7 = GetPartSkeletalMeshComponent()
   > n8 = TransformFromBoneSpace!(
    execute
    self:Skinned Mesh Component Object Reference=@n7
    BoneName:Name=dyn_tricep_measure_r
    InPosition:Vector="0, 0, 0"
    InRotation:Rotator="0, 0, 0"
   ) -> (then, OutPosition:Vector, OutRotation:Rotator)
   > n9 = TransformFromBoneSpace(self=@n7, BoneName=dyn_lat_measure_r, InPosition="0, 0, 0", InRotation="0, 0, 0")
   > Set arm_rot_dn_r(arm_rot_dn_r:double=MapRangeClamped(Value="Distance (Vector)"!(V1:Vector=@n8, V2:Vector=@n9) -> (ReturnValue:double), InRangeA=13, InRangeB=6)) -> (Output_Get:double)
   n10 = GetOwnerSkeletalMeshComponent()
   n11 = BreakRotator(InRot=BreakTransform(InTransform=GetDeltaTransformFromRefPose(self=@n10, BoneName=upperarm_r, BaseName=spine_05)).Rotation)
   n12 = MapRangeClamped(Value=@n11.Yaw, InRangeA=0, InRangeB=-40)
   > Set arm_rot_bk_r(arm_rot_bk_r:double=@n12) -> (Output_Get:double)
   n13 = MapRangeClamped(Value=@n11.Yaw, InRangeA=30, InRangeB=140)
   > Set arm_rot_up_r(arm_rot_up_r:double=@n13) -> (Output_Get:double)
   n14 = BreakTransform(InTransform=GetDeltaTransformFromRefPose(self=@n10, BoneName=upperarm_l, BaseName=spine_04))
   > Set arm_rot_up_l(arm_rot_up_l:double=MapRangeClamped(Value=BreakRotator(InRot=@n14.Rotation).Yaw, InRangeA=20, InRangeB=100)) -> (Output_Get:double)
   n15 = BreakTransform(InTransform=GetDeltaTransformFromRefPose(self=@n10, BoneName=clavicle_r, BaseName=spine_05))
   > Set clavicle_rot_up_r(clavicle_rot_up_r:double=MapRangeClamped(Value=BreakRotator(InRot=@n15.Rotation).Pitch, InRangeA=10, InRangeB=40)) -> (Output_Get:double)
   n16 = GetOwnerSkeletalMeshComponent()
   n17 = BreakTransform(InTransform=GetDeltaTransformFromRefPose(self=@n16, BoneName=thigh_l, BaseName=pelvis))
   > Set cloak_aim_alpha_l(cloak_aim_alpha_l:double=MapRangeClamped(Value=BreakRotator(InRot=@n17.Rotation).Yaw, InRangeA=-100, InRangeB=-25)) -> (Output_Get:double)
   n18 = BreakRotator(InRot=BreakTransform(InTransform=GetDeltaTransformFromRefPose(self=@n16, BoneName=thigh_r, BaseName=pelvis)).Rotation)
   > Set thigh_rot_up_r(thigh_rot_up_r:double=MapRangeClamped(Value=@n18.Yaw, InRangeA=-10, InRangeB=-130)) -> (Output_Get:double)
   > Set cloak_aim_alpha_r(cloak_aim_alpha_r:double=MapRangeClamped(Value=@n18.Yaw, InRangeA=-100, InRangeB=-25)) -> (Output_Get:double)
   n19 = GetOwnerSkeletalMeshComponent()
   n20 = BreakTransform(InTransform=GetDeltaTransformFromRefPose(self=@n19, BoneName=hand_l, BaseName=lowerarm_twist_01_l))
   > Set hand_rot_l(hand_rot_l:double=MapRangeClamped(Value=BreakRotator(InRot=@n20.Rotation).Yaw, InRangeA=0, InRangeB=80)) -> (Output_Get:double)
   n21 = BreakRotator(InRot=BreakTransform(InTransform=GetDeltaTransformFromRefPose(self=@n19, BoneName=lowerarm_twist_01_l, BaseName=lowerarm_l)).Rotation)
   > Set hand_twist_l(hand_twist_l:double=MapRangeClamped(Value="Absolute (Float)"(A=@n21), InRangeA=20, InRangeB=120)) -> (Output_Get:double)
   n22 = BreakTransform(InTransform=GetDeltaTransformFromRefPose(self=GetOwnerSkeletalMeshComponent(), BoneName=spine_05, BaseName=spine_02))
   > Set chest_twist(chest_twist:double=MapRangeClamped(Value=BreakRotator(InRot=@n22.Rotation), InRangeA=0, InRangeB=20)) -> (Output_Get:double)
   n23 = Branch(execute=@n6.then_1, Condition=Get bClothEnabled())
   > n24 = Sequence!(execute) -> (then_0, then_1)
   n25 = Gravity Override State Function DEPRECATED(
    execute=[@n24.then_0, @n23.else]
    CharacterPartAnimInstance=Self-Reference()
    GameTimeInSeconds=GetGameTimeInSeconds()
    GravityOverrideParams=Get Gravity Override Params()
   )
   > Set GravityOverride(GravityOverride=@n25)
   n26 = Branch(execute=@n24.then_1, Condition=Get isWearBackbling())
   > n27 = Cloth Controller Function DEPRECATED!(
    execute
    CharacterPartAnimInstance:Custom Character Part Anim Instance Object Reference=Self-Reference()
    ClothingAssetName:String=Lancelot_Clothing
    Cloth Paramaters:Cloth Params Struct Structure=Get Cloth Paramaters_Wearing_Pack() -> (Cloth Paramaters_Wearing_Pack:Cloth Params Struct Structure)
    Cloth Core Settings:Cloth Core Setting Struct Structure=Get Cloth Core Settings_0() -> (Cloth Core Settings_0:Cloth Core Setting Struct Structure)
    ClothUpAxis:Vector="1,0,0"
    OnGround_Moving_SpeedLimit:double=2
    OnGround_Moving_AngularVel_Max:double=0.05
    OnGround_Moving_FictitiousVel_Max:double=0.6
    LinVelScale_FwdBwd_Clamp:double=0
    LinVelScale_UpDwn_Clamp:double=0.7
    DeltaTime:double=Get DeltaTimeLocal() -> (DeltaTimeLocal:double)
    GameTimeInSeconds:double=GetGameTimeInSeconds()
    WindGustsParams:Wind Gust Params Struct Structure=Get Wind Gust Params() -> (Wind Gust Params:Wind Gust Params Struct Structure)
    WindGustLocalSettings:Wind Gust Local Setting Struct Structure=Get Gust Local Settings() -> (Gust Local Settings:Wind Gust Local Setting Struct Structure)
   ) -> (
    then
    CurrentStateOut:String
    Cloth Settings Out:Cloth Setting Struct Structure
    Cloth Core Settings Out:Cloth Core Setting Struct Structure
    WindGustSettings Out:Wind Gust Setting Struct Structure
    WindGustLocalSettings Out:Wind Gust Local Setting Struct Structure
    GustStrength - WindVelocityOut:Vector
    LinearVelVector-DELETE LATER:Vector
   )
   > Set Gust Local Settings(Gust Local Settings:Wind Gust Local Setting Struct Structure=@n27.WindGustLocalSettings Out) -> (Output_Get:Wind Gust Local Setting Struct Structure)
   n28 = Cloth Controller Function DEPRECATED(
    execute=@n26.else
    CharacterPartAnimInstance=Self-Reference()
    ClothingAssetName=Lancelot_Clothing
    Cloth Paramaters=Get Cloth Paramaters_0() -> (Cloth Paramaters_0:Cloth Params Struct Structure)
    Cloth Core Settings=Get Cloth Core Settings_0()
    ClothUpAxis="1,0,0"
    OnGround_Moving_SpeedLimit=2
    OnGround_Moving_AngularVel_Max=0.05
    OnGround_Moving_FictitiousVel_Max=0.6
    LinVelScale_FwdBwd_Clamp=0
    LinVelScale_UpDwn_Clamp=0.7
    DeltaTime=Get DeltaTimeLocal()
    GameTimeInSeconds=GetGameTimeInSeconds()
    WindGustsParams=Get Wind Gust Params()
    WindGustLocalSettings=Get Gust Local Settings()
   )
   > Set Gust Local Settings(Gust Local Settings=@n28.WindGustLocalSettings Out)
   n29 = Branch(execute=@n6.then_2, Condition=Get bIsParachuteOpened())
   > Set arm_rot_up_r(arm_rot_up_r=0)
   > Set arm_rot_bk_r(arm_rot_bk_r=0)
   n30 = BreakTransform(InTransform=GetDeltaTransformFromRefPose(self=GetOwnerSkeletalMeshComponent(), BoneName=upperarm_r, BaseName=spine_05))
   > Set def_cape_parachute(def_cape_parachute:double=MapRangeClamped(Value=BreakRotator(InRot=@n30.Rotation).Pitch, InRangeA=10, InRangeB=80)) -> (Output_Get:double)
   Set arm_rot_up_r(execute=@n29.else, arm_rot_up_r=@n13)
   > Set arm_rot_bk_r(arm_rot_bk_r=@n12)
   > Set def_cape_parachute(def_cape_parachute=0)
   n31 = Get Anim_Inst() -> (Anim_Inst:Anim Instance Object Reference)
   Is Valid(exec=@n6.then_3, InputObject=@n31)
   > n32 = Branch(Condition=Get bIsInFrontEnd())
   > Set isIdleEmotePlaying(isIdleEmotePlaying:bool=Montage_IsPlaying(self=@n31, Montage=/BRCosmetics/Animation/Game/MainPlayer/Menu/BR/CustomSkins/_s21/LancelotM/Lancelot_Male_Idle_Montage)) -> (Output_Get:bool)
   Set isIdleEmotePlaying(execute=@n32.else, isIdleEmotePlaying=false)
   DrawVectorFromJointFunction!(
    execute
    SkelMeshComponent:Scene Component Object Reference
    JointName:Name=pelvis
    LineEnd:Vector=Break Cloth Setting Struct!(ClothSettingStruct:"Cloth Setting Struct Structure (by ref)"=@n28.Cloth Settings Out) -> (
     LinearVelocityScale_7_#1:Vector=~
     AngularVelocityScale_6_#2:float=~
     FictitiousAngularScale_10_#3:float=~
     GravityOverride_15_#4:Vector=~
     DragCoefficient_23_#5:Vector 2D Structure
     LiftCoefficient_24_#6:Vector 2D Structure
     JointName_27_#7:Name
    ) {is=title K2Node_BreakStruct}.GravityOverride_15_#4
    JointOffset:Vector="0, 0, 0"
    LineColor:Linear Color Structure=(R=1,G=0,B=0,A=1)
    Thickness:double=2
    Foreground:EDrawDebugSceneDepthPriorityGroup Enum=Foreground
   ) -> (then)
  ]}
]
variables [
 Anim_Inst {type=AnimInstance, default=null}
 arm_rot_bk_r {type=number}
 arm_rot_dn_r {type=number}
 arm_rot_up_l {type=number}
 arm_rot_up_r {type=number}
 chest_twist {type=number}
 clavicle_rot_up_r {type=number}
 cloak_aim_alpha_l {type=number}
 cloak_aim_alpha_r {type=number}
 Cloth Core Settings_0 {
  type=ClothCoreSettingStruct
  default {
   collisionThickness=0
   dampingCoefficient=0.1
   frictionCoefficient=0.3
   localDampingCoefficient=0.1
   numberIterations=2
   numberSubsteps=1
   selfCollisionThickness=0
   animDriveDamping=<0 1>
   animDriveStiffness=<0 1>
   areaStiffness=<1 1>
   bendingStiffness=<1 1>
   edgeStiffness=<1 1>
   tetherStiffness=<1 1>
  }
 }
 Cloth Paramaters_0 {
  type=ClothParamsStruct
  default {
   dBNO {
    angularVelocityScale=0.6
    fictitiousAngularScale=0.1
    jointName=pelvis
    dragCoefficient=<0.2 1>
    gravityOverride=<-980 400 0>
    liftCoefficient=<0 0.5>
    linearVelocityScale=<0.3 0.3 0.1>
   }
   dBNO_Carried {
    angularVelocityScale=0.6
    fictitiousAngularScale=0.1
    jointName=pelvis
    dragCoefficient=<0.2 1>
    gravityOverride=<-980 -200 0>
    liftCoefficient=<0 0.5>
    linearVelocityScale=<0.3 0.3 0.1>
   }
   dBNO_Impulse {
    angularVelocityScale=0
    fictitiousAngularScale=0
    jointName=pelvis
    dragCoefficient=<0 0.2>
    gravityOverride=<-3000 -2000 0>
    liftCoefficient=<0 0.1>
    linearVelocityScale=<0.3 0.1 0.3>
   }
   dBNO_RollBwd {
    angularVelocityScale=0
    fictitiousAngularScale=0
    jointName=pelvis
    dragCoefficient=<0 0.2>
    gravityOverride=<-2000 0 0>
    liftCoefficient=<0 0.1>
    linearVelocityScale=<0.3 0.1 0.3>
   }
   dBNO_RollFwd {
    angularVelocityScale=0
    fictitiousAngularScale=0
    jointName=pelvis
    dragCoefficient=<0 0.2>
    gravityOverride=<-980 0 0>
    liftCoefficient=<0 0.1>
    linearVelocityScale=<0.3 0.1 0.3>
   }
   dBNO_Tumbling {
    angularVelocityScale=0
    fictitiousAngularScale=0
    jointName=pelvis
    dragCoefficient=<0 0.2>
    gravityOverride=<-2500 -300 0>
    liftCoefficient=<0 0.1>
    linearVelocityScale=<0.3 0.1 0.3>
   }
   emoteOrMelee {
    angularVelocityScale=0.2
    fictitiousAngularScale=0.1
    jointName=pelvis
    dragCoefficient=<0 0.3>
    gravityOverride=<-980 -500 0>
    liftCoefficient=<0 0.2>
    linearVelocityScale=<0.05 0.05 0.05>
   }
   front_End {
    angularVelocityScale=0.1
    fictitiousAngularScale=0.2
    jointName=root
    dragCoefficient=<0.2 0.2>
    gravityOverride=<0 -200 -980>
    liftCoefficient=<0.1 0.1>
    linearVelocityScale=<0.1 0.1 0.1>
   }
   onGround_Crouch_Moving {
    angularVelocityScale=0.1
    fictitiousAngularScale=0.3
    jointName=root
    dragCoefficient=<2 2>
    gravityOverride=<0 0 -980>
    liftCoefficient=<0.6 0.6>
    linearVelocityScale=<0.6 0.6 0.5>
   }
   onGround_Falling {
    angularVelocityScale=0.6
    fictitiousAngularScale=0
    jointName=root
    dragCoefficient=<0 0.2>
    gravityOverride=<0 -500 -980>
    liftCoefficient=<0 0.1>
    linearVelocityScale=<2 1 0.3>
   }
   onGround_Moving {
    angularVelocityScale=0.6
    fictitiousAngularScale=0.1
    jointName=root
    dragCoefficient=<0 0.3>
    gravityOverride=<0 -100 -980>
    liftCoefficient=<0 0.2>
    linearVelocityScale=<0.6 0.8 0.8>
   }
   onGround_Standing {
    angularVelocityScale=0.05
    fictitiousAngularScale=0.6
    jointName=root
    dragCoefficient=<0 2>
    gravityOverride=<0 -100 -980>
    liftCoefficient=<0 0.5>
    linearVelocityScale=<0.6 0.8 0.8>
   }
   rail_Grinding {
    angularVelocityScale=0.2
    fictitiousAngularScale=0
    jointName=root
    dragCoefficient=<0.1 0.2>
    gravityOverride=<50 -500 -400>
    liftCoefficient=<0 0.5>
    linearVelocityScale=<0.1 0.1 0.1>
   }
   riding_Moving {
    angularVelocityScale=0.05
    fictitiousAngularScale=1
    jointName=root
    dragCoefficient=<0.5 0.5>
    gravityOverride=<0 0 -980>
    liftCoefficient=<0.5 0.5>
    linearVelocityScale=<0.4 0.6 0.3>
   }
   riding_Moving_Targeting {
    angularVelocityScale=0.05
    fictitiousAngularScale=0.7
    jointName=root
    dragCoefficient=<0.5 0.5>
    gravityOverride=<0 0 -980>
    liftCoefficient=<0.5 0.5>
    linearVelocityScale=<0.2 0.6 0.2>
   }
   riding_Standing {
    angularVelocityScale=0.05
    fictitiousAngularScale=1
    jointName=root
    dragCoefficient=<0.5 0.5>
    gravityOverride=<0 -450 -980>
    liftCoefficient=<0.5 0.5>
    linearVelocityScale=<0.6 0.6 0.3>
   }
   skydive_DivingDown {
    angularVelocityScale=0.1
    fictitiousAngularScale=0
    jointName=pelvis
    dragCoefficient=<0 1.5>
    gravityOverride=<-450 -100 0>
    liftCoefficient=<0 0.3>
    linearVelocityScale=<0.1 0.1 0.1>
   }
   skydive_FloatingDown {
    angularVelocityScale=0.1
    fictitiousAngularScale=0
    jointName=root
    dragCoefficient=<0 2.5>
    gravityOverride=<0 0 980>
    liftCoefficient=<0 0.4>
    linearVelocityScale=<0.2 0.2 0.1>
   }
   skydive_GoingUp {
    angularVelocityScale=0.1
    fictitiousAngularScale=0
    jointName=root
    dragCoefficient=<0 0>
    gravityOverride=<0 -3000 3000>
    liftCoefficient=<0 0>
    linearVelocityScale=<0.1 0.1 0.05>
   }
   skydive_Parachuting {
    angularVelocityScale=0.1
    fictitiousAngularScale=0
    jointName=root
    dragCoefficient=<0 2.5>
    gravityOverride=<0 0 -980>
    liftCoefficient=<0 0.5>
    linearVelocityScale=<0.2 0.2 0.2>
   }
   skydive_Surfing {
    angularVelocityScale=0.6
    fictitiousAngularScale=0.1
    jointName=root
    dragCoefficient=<1.5 1.5>
    gravityOverride=<0 -600 -500>
    liftCoefficient=<1 1>
    linearVelocityScale=<0.1 0.1 0.3>
   }
   swimming {
    angularVelocityScale=0.6
    fictitiousAngularScale=0
    jointName=root
    dragCoefficient=<0 2>
    gravityOverride=<0 -600 400>
    liftCoefficient=<0 0.4>
    linearVelocityScale=<0.6 0.2 0.2>
   }
   tactical_Sprint {
    angularVelocityScale=0.6
    fictitiousAngularScale=0.1
    jointName=root
    dragCoefficient=<2 2>
    gravityOverride=<0 0 -980>
    liftCoefficient=<0.5 0.5>
    linearVelocityScale=<0.6 0.6 0.5>
   }
   ziplining {
    angularVelocityScale=0.6
    fictitiousAngularScale=0
    jointName=root
    dragCoefficient=<0 1.5>
    gravityOverride=<0 0 -980>
    liftCoefficient=<0 0.3>
    linearVelocityScale=<0.4 0.4 0.5>
   }
  }
 }
 Cloth Paramaters_Wearing_Pack {
  type=ClothParamsStruct
  default {
   dBNO {
    angularVelocityScale=0.6
    fictitiousAngularScale=0.1
    jointName=pelvis
    dragCoefficient=<0.2 1>
    gravityOverride=<-980 400 0>
    liftCoefficient=<0 0.5>
    linearVelocityScale=<0.3 0.3 0.1>
   }
   dBNO_Carried {
    angularVelocityScale=0.6
    fictitiousAngularScale=0.1
    jointName=pelvis
    dragCoefficient=<0.2 1>
    gravityOverride=<-980 -200 0>
    liftCoefficient=<0 0.5>
    linearVelocityScale=<0.3 0.3 0.1>
   }
   dBNO_Impulse {
    angularVelocityScale=0
    fictitiousAngularScale=0
    jointName=pelvis
    dragCoefficient=<0 0.2>
    gravityOverride=<-3000 -2000 0>
    liftCoefficient=<0 0.1>
    linearVelocityScale=<0.3 0.1 0.3>
   }
   dBNO_RollBwd {
    angularVelocityScale=0
    fictitiousAngularScale=0
    jointName=pelvis
    dragCoefficient=<0 0.2>
    gravityOverride=<-2000 0 0>
    liftCoefficient=<0 0.1>
    linearVelocityScale=<0.3 0.1 0.3>
   }
   dBNO_RollFwd {
    angularVelocityScale=0
    fictitiousAngularScale=0
    jointName=pelvis
    dragCoefficient=<0 0.2>
    gravityOverride=<-980 0 0>
    liftCoefficient=<0 0.1>
    linearVelocityScale=<0.3 0.1 0.3>
   }
   dBNO_Tumbling {
    angularVelocityScale=0
    fictitiousAngularScale=0
    jointName=pelvis
    dragCoefficient=<0 0.2>
    gravityOverride=<-2500 -300 0>
    liftCoefficient=<0 0.1>
    linearVelocityScale=<0.3 0.1 0.3>
   }
   emoteOrMelee {
    angularVelocityScale=0.2
    fictitiousAngularScale=0.1
    jointName=pelvis
    dragCoefficient=<0 0.3>
    gravityOverride=<-980 -500 0>
    liftCoefficient=<0 0.2>
    linearVelocityScale=<0.05 0.05 0.05>
   }
   front_End {
    angularVelocityScale=0.1
    fictitiousAngularScale=0.2
    jointName=root
    dragCoefficient=<0.2 0.2>
    gravityOverride=<0 -200 -980>
    liftCoefficient=<0.1 0.1>
    linearVelocityScale=<0.1 0.1 0.1>
   }
   onGround_Crouch_Moving {
    angularVelocityScale=0.1
    fictitiousAngularScale=0.3
    jointName=root
    dragCoefficient=<2 2>
    gravityOverride=<0 0 -980>
    liftCoefficient=<0.6 0.6>
    linearVelocityScale=<0.5 0.5 0.5>
   }
   onGround_Falling {
    angularVelocityScale=0.6
    fictitiousAngularScale=0
    jointName=root
    dragCoefficient=<0 0.2>
    gravityOverride=<0 -500 -980>
    liftCoefficient=<0 0.1>
    linearVelocityScale=<2 1 0.3>
   }
   onGround_Moving {
    angularVelocityScale=0.6
    fictitiousAngularScale=0.1
    jointName=root
    dragCoefficient=<0 0.3>
    gravityOverride=<0 -100 -980>
    liftCoefficient=<0 0.2>
    linearVelocityScale=<0.3 0.3 0.3>
   }
   onGround_Standing {
    angularVelocityScale=0.05
    fictitiousAngularScale=0.6
    jointName=root
    dragCoefficient=<0 2>
    gravityOverride=<0 -100 -980>
    liftCoefficient=<0 0.5>
    linearVelocityScale=<0.3 0.3 0.3>
   }
   rail_Grinding {
    angularVelocityScale=0.2
    fictitiousAngularScale=0
    jointName=root
    dragCoefficient=<0.1 0.2>
    gravityOverride=<50 -500 -400>
    liftCoefficient=<0 0.5>
    linearVelocityScale=<0.1 0.1 0.1>
   }
   riding_Moving {
    angularVelocityScale=0.05
    fictitiousAngularScale=1
    jointName=root
    dragCoefficient=<0.5 0.5>
    gravityOverride=<0 0 -980>
    liftCoefficient=<0.5 0.5>
    linearVelocityScale=<0.4 0.6 0.3>
   }
   riding_Moving_Targeting {
    angularVelocityScale=0.05
    fictitiousAngularScale=0.7
    jointName=root
    dragCoefficient=<0.5 0.5>
    gravityOverride=<0 0 -980>
    liftCoefficient=<0.5 0.5>
    linearVelocityScale=<0.2 0.6 0.2>
   }
   riding_Standing {
    angularVelocityScale=0.05
    fictitiousAngularScale=1
    jointName=root
    dragCoefficient=<0.5 0.5>
    gravityOverride=<0 -450 -980>
    liftCoefficient=<0.5 0.5>
    linearVelocityScale=<0.6 0.6 0.3>
   }
   skydive_DivingDown {
    angularVelocityScale=0.1
    fictitiousAngularScale=0
    jointName=pelvis
    dragCoefficient=<0 1.5>
    gravityOverride=<-450 -100 0>
    liftCoefficient=<0 0.3>
    linearVelocityScale=<0.1 0.1 0.1>
   }
   skydive_FloatingDown {
    angularVelocityScale=0.1
    fictitiousAngularScale=0
    jointName=root
    dragCoefficient=<0 2.5>
    gravityOverride=<0 0 980>
    liftCoefficient=<0 0.4>
    linearVelocityScale=<0.2 0.2 0.1>
   }
   skydive_GoingUp {
    angularVelocityScale=0.1
    fictitiousAngularScale=0
    jointName=root
    dragCoefficient=<0 0>
    gravityOverride=<0 -3000 3000>
    liftCoefficient=<0 0>
    linearVelocityScale=<0.1 0.1 0.05>
   }
   skydive_Parachuting {
    angularVelocityScale=0.1
    fictitiousAngularScale=0
    jointName=root
    dragCoefficient=<0 2.5>
    gravityOverride=<0 0 -980>
    liftCoefficient=<0 0.5>
    linearVelocityScale=<0.2 0.2 0.2>
   }
   skydive_Surfing {
    angularVelocityScale=0.6
    fictitiousAngularScale=0.1
    jointName=root
    dragCoefficient=<1.5 1.5>
    gravityOverride=<0 -600 -500>
    liftCoefficient=<1 1>
    linearVelocityScale=<0.1 0.1 0.3>
   }
   swimming {
    angularVelocityScale=0.6
    fictitiousAngularScale=0
    jointName=root
    dragCoefficient=<0 2>
    gravityOverride=<0 -600 400>
    liftCoefficient=<0 0.4>
    linearVelocityScale=<0.6 0.2 0.2>
   }
   tactical_Sprint {
    angularVelocityScale=0.6
    fictitiousAngularScale=0.1
    jointName=root
    dragCoefficient=<2 2>
    gravityOverride=<0 0 -980>
    liftCoefficient=<0.5 0.5>
    linearVelocityScale=<0.6 0.6 0.5>
   }
   ziplining {
    angularVelocityScale=0.6
    fictitiousAngularScale=0
    jointName=root
    dragCoefficient=<0 1.5>
    gravityOverride=<0 0 -980>
    liftCoefficient=<0 0.3>
    linearVelocityScale=<0.4 0.4 0.5>
   }
  }
 }
 def_cape_parachute {type=number}
 DeltaTimeLocal {type=number}
 Gravity Override Params {
  type=GravityOverrideParamsStruct
  default {
   dBNO_Carried {gravityOverride=<-980 980 -500>}
   onGround_Standing {windAmplitude=<50 -50 10>, windFrequency=<0.15 0.8 0.5>, windOffset=<0 -80 -15>}
  }
 }
 GravityOverride {type=Vector}
 Gust Local Settings {type=WindGustLocalSettingStruct, default {elapsedTime=0, finalGustStrengthLocal=0, gustDuration=0, gustStrength=0, timeBetweenGusts=0}}
 hand_rot_l {type=number}
 hand_twist_l {type=number}
 isIdleEmotePlaying {type=boolean}
 isWearBackbling {type=boolean}
 isWearCape {type=boolean}
 Player_Pawn {type=FortPlayerPawn}
 Skel_Mesh {type=SkeletalMeshComponent}
 thigh_rot_up_r {type=number}
 Wind Gust Params {
  type=WindGustParamsStruct
  default {
   dBNO {
    maxDownInterpSpeed=5
    maxGustDuration=2
    maxGustStrength=7
    maxTimeBetweenGusts=7
    maxUpInterpSpeed=5
    minDownInterpSpeed=2
    minGustDuration=0.5
    minGustStrength=3
    minTimeBetweenGusts=0
    minUpInterpSpeed=2
    windTurbulenceScalarA=0.1
    windTurbulenceScalarB=1
    windAmplitude=<0 0 0>
    windFrequency=<0 0 0>
    windOffset=<0 0 0>
   }
   dBNO_Carried {
    maxDownInterpSpeed=5
    maxGustDuration=2
    maxGustStrength=7
    maxTimeBetweenGusts=7
    maxUpInterpSpeed=5
    minDownInterpSpeed=2
    minGustDuration=0.5
    minGustStrength=3
    minTimeBetweenGusts=0
    minUpInterpSpeed=2
    windTurbulenceScalarA=0.1
    windTurbulenceScalarB=1
    windAmplitude=<0 0 0>
    windFrequency=<0 0 0>
    windOffset=<0 0 0>
   }
   dBNO_Impulse {
    maxDownInterpSpeed=5
    maxGustDuration=2
    maxGustStrength=7
    maxTimeBetweenGusts=7
    maxUpInterpSpeed=5
    minDownInterpSpeed=2
    minGustDuration=0.5
    minGustStrength=3
    minTimeBetweenGusts=0
    minUpInterpSpeed=2
    windTurbulenceScalarA=0.1
    windTurbulenceScalarB=1
    windAmplitude=<0 0 0>
    windFrequency=<0 0 0>
    windOffset=<0 0 0>
   }
   dBNO_RollBwd {
    maxDownInterpSpeed=5
    maxGustDuration=2
    maxGustStrength=7
    maxTimeBetweenGusts=7
    maxUpInterpSpeed=5
    minDownInterpSpeed=2
    minGustDuration=0.5
    minGustStrength=3
    minTimeBetweenGusts=0
    minUpInterpSpeed=2
    windTurbulenceScalarA=0.1
    windTurbulenceScalarB=1
    windAmplitude=<0 0 0>
    windFrequency=<0 0 0>
    windOffset=<0 0 0>
   }
   dBNO_RollFwd {
    maxDownInterpSpeed=5
    maxGustDuration=2
    maxGustStrength=7
    maxTimeBetweenGusts=7
    maxUpInterpSpeed=5
    minDownInterpSpeed=2
    minGustDuration=0.5
    minGustStrength=3
    minTimeBetweenGusts=0
    minUpInterpSpeed=2
    windTurbulenceScalarA=0.1
    windTurbulenceScalarB=1
    windAmplitude=<0 0 0>
    windFrequency=<0 0 0>
    windOffset=<0 0 0>
   }
   dBNO_Tumbling {
    maxDownInterpSpeed=5
    maxGustDuration=2
    maxGustStrength=7
    maxTimeBetweenGusts=7
    maxUpInterpSpeed=5
    minDownInterpSpeed=2
    minGustDuration=0.5
    minGustStrength=3
    minTimeBetweenGusts=0
    minUpInterpSpeed=2
    windTurbulenceScalarA=0.1
    windTurbulenceScalarB=1
    windAmplitude=<0 0 0>
    windFrequency=<0 0 0>
    windOffset=<0 0 0>
   }
   emoteOrMelee {
    maxDownInterpSpeed=5
    maxGustDuration=2
    maxGustStrength=5
    maxTimeBetweenGusts=7
    maxUpInterpSpeed=5
    minDownInterpSpeed=2
    minGustDuration=0.5
    minGustStrength=3
    minTimeBetweenGusts=0
    minUpInterpSpeed=2
    windTurbulenceScalarA=0.1
    windTurbulenceScalarB=0.6
    windAmplitude=<30 25 50>
    windFrequency=<2 3 5>
    windOffset=<-300 -150 0>
   }
   front_End {
    maxDownInterpSpeed=5
    maxGustDuration=2
    maxGustStrength=4
    maxTimeBetweenGusts=7
    maxUpInterpSpeed=5
    minDownInterpSpeed=2
    minGustDuration=0.5
    minGustStrength=2
    minTimeBetweenGusts=0
    minUpInterpSpeed=2
    windTurbulenceScalarA=0.1
    windTurbulenceScalarB=1
    windAmplitude=<50 150 180>
    windFrequency=<10 4 6>
    windOffset=<0 -300 -200>
   }
   onGround_Crouch_Moving {
    maxDownInterpSpeed=5
    maxGustDuration=2
    maxGustStrength=7
    maxTimeBetweenGusts=7
    maxUpInterpSpeed=5
    minDownInterpSpeed=2
    minGustDuration=0.5
    minGustStrength=3
    minTimeBetweenGusts=0
    minUpInterpSpeed=2
    windTurbulenceScalarA=0.1
    windTurbulenceScalarB=1
    windAmplitude=<100 25 75>
    windFrequency=<1.5 0.5 7.5>
    windOffset=<0 -100 -175>
   }
   onGround_Falling {
    maxDownInterpSpeed=5
    maxGustDuration=2
    maxGustStrength=7
    maxTimeBetweenGusts=7
    maxUpInterpSpeed=5
    minDownInterpSpeed=2
    minGustDuration=0.5
    minGustStrength=3
    minTimeBetweenGusts=0
    minUpInterpSpeed=2
    windTurbulenceScalarA=0.1
    windTurbulenceScalarB=1
    windAmplitude=<0 0 0>
    windFrequency=<0 0 0>
    windOffset=<0 0 0>
   }
   onGround_Moving {
    maxDownInterpSpeed=5
    maxGustDuration=8
    maxGustStrength=7
    maxTimeBetweenGusts=8
    maxUpInterpSpeed=5
    minDownInterpSpeed=2
    minGustDuration=0.2
    minGustStrength=3
    minTimeBetweenGusts=1
    minUpInterpSpeed=2
    windTurbulenceScalarA=0.1
    windTurbulenceScalarB=0.8
    windAmplitude=<30 65 42>
    windFrequency=<5 6 8.6>
    windOffset=<30 -250 -300>
   }
   onGround_Standing {
    maxDownInterpSpeed=5
    maxGustDuration=8
    maxGustStrength=7
    maxTimeBetweenGusts=8
    maxUpInterpSpeed=5
    minDownInterpSpeed=2
    minGustDuration=0.2
    minGustStrength=2
    minTimeBetweenGusts=1
    minUpInterpSpeed=2
    windTurbulenceScalarA=0.1
    windTurbulenceScalarB=0.8
    windAmplitude=<75 24 28>
    windFrequency=<1.5 2.3 3.7>
    windOffset=<75 -100 -150>
   }
   rail_Grinding {
    maxDownInterpSpeed=5
    maxGustDuration=8
    maxGustStrength=7
    maxTimeBetweenGusts=8
    maxUpInterpSpeed=5
    minDownInterpSpeed=2
    minGustDuration=0.2
    minGustStrength=3
    minTimeBetweenGusts=1
    minUpInterpSpeed=2
    windTurbulenceScalarA=0.1
    windTurbulenceScalarB=0.8
    windAmplitude=<20 65 42>
    windFrequency=<5 6 8.6>
    windOffset=<50 -500 -400>
   }
   riding_Moving {
    maxDownInterpSpeed=5
    maxGustDuration=2
    maxGustStrength=7
    maxTimeBetweenGusts=7
    maxUpInterpSpeed=5
    minDownInterpSpeed=2
    minGustDuration=0.5
    minGustStrength=3
    minTimeBetweenGusts=0
    minUpInterpSpeed=2
    windTurbulenceScalarA=0.1
    windTurbulenceScalarB=1
    windAmplitude=<0 0 0>
    windFrequency=<0 0 0>
    windOffset=<0 0 0>
   }
   riding_Moving_Targeting {
    maxDownInterpSpeed=5
    maxGustDuration=2
    maxGustStrength=7
    maxTimeBetweenGusts=7
    maxUpInterpSpeed=5
    minDownInterpSpeed=2
    minGustDuration=0.5
    minGustStrength=3
    minTimeBetweenGusts=0
    minUpInterpSpeed=2
    windTurbulenceScalarA=0.1
    windTurbulenceScalarB=1
    windAmplitude=<0 0 0>
    windFrequency=<0 0 0>
    windOffset=<0 0 0>
   }
   riding_Standing {
    maxDownInterpSpeed=5
    maxGustDuration=2
    maxGustStrength=7
    maxTimeBetweenGusts=7
    maxUpInterpSpeed=5
    minDownInterpSpeed=2
    minGustDuration=0.5
    minGustStrength=3
    minTimeBetweenGusts=0
    minUpInterpSpeed=2
    windTurbulenceScalarA=0.1
    windTurbulenceScalarB=1
    windAmplitude=<0 0 0>
    windFrequency=<0 0 0>
    windOffset=<0 0 0>
   }
   skydive_DivingDown {
    maxDownInterpSpeed=5
    maxGustDuration=8
    maxGustStrength=7
    maxTimeBetweenGusts=8
    maxUpInterpSpeed=5
    minDownInterpSpeed=2
    minGustDuration=0.2
    minGustStrength=3
    minTimeBetweenGusts=1
    minUpInterpSpeed=2
    windTurbulenceScalarA=0.1
    windTurbulenceScalarB=4
    windAmplitude=<0 25 35>
    windFrequency=<0 4 10>
    windOffset=<-450 -100 100>
   }
   skydive_FloatingDown {
    maxDownInterpSpeed=5
    maxGustDuration=8
    maxGustStrength=1
    maxTimeBetweenGusts=8
    maxUpInterpSpeed=5
    minDownInterpSpeed=2
    minGustDuration=0.2
    minGustStrength=0
    minTimeBetweenGusts=1
    minUpInterpSpeed=2
    windTurbulenceScalarA=0.5
    windTurbulenceScalarB=1
    windAmplitude=<50 20 35>
    windFrequency=<8 2 3>
    windOffset=<0 -300 200>
   }
   skydive_GoingUp {
    maxDownInterpSpeed=5
    maxGustDuration=8
    maxGustStrength=7
    maxTimeBetweenGusts=8
    maxUpInterpSpeed=5
    minDownInterpSpeed=2
    minGustDuration=0.2
    minGustStrength=3
    minTimeBetweenGusts=1
    minUpInterpSpeed=2
    windTurbulenceScalarA=0.1
    windTurbulenceScalarB=0.8
    windAmplitude=<0 0 0>
    windFrequency=<0 0 0>
    windOffset=<0 -500 300>
   }
   skydive_Parachuting {
    maxDownInterpSpeed=5
    maxGustDuration=8
    maxGustStrength=1
    maxTimeBetweenGusts=8
    maxUpInterpSpeed=5
    minDownInterpSpeed=2
    minGustDuration=5
    minGustStrength=0
    minTimeBetweenGusts=5
    minUpInterpSpeed=2
    windTurbulenceScalarA=0.2
    windTurbulenceScalarB=1
    windAmplitude=<25 50 75>
    windFrequency=<8 6 2>
    windOffset=<-150 -400 -150>
   }
   skydive_Surfing {
    maxDownInterpSpeed=5
    maxGustDuration=2
    maxGustStrength=7
    maxTimeBetweenGusts=7
    maxUpInterpSpeed=5
    minDownInterpSpeed=2
    minGustDuration=0.5
    minGustStrength=3
    minTimeBetweenGusts=0
    minUpInterpSpeed=2
    windTurbulenceScalarA=0.1
    windTurbulenceScalarB=1
    windAmplitude=<0 0 0>
    windFrequency=<0 0 0>
    windOffset=<0 0 0>
   }
   swimming {
    maxDownInterpSpeed=5
    maxGustDuration=2
    maxGustStrength=7
    maxTimeBetweenGusts=7
    maxUpInterpSpeed=5
    minDownInterpSpeed=2
    minGustDuration=0.5
    minGustStrength=3
    minTimeBetweenGusts=0
    minUpInterpSpeed=2
    windTurbulenceScalarA=0.1
    windTurbulenceScalarB=1
    windAmplitude=<10 25 20>
    windFrequency=<1.5 0.5 7.5>
    windOffset=<0 -150 50>
   }
   tactical_Sprint {
    maxDownInterpSpeed=5
    maxGustDuration=2
    maxGustStrength=7
    maxTimeBetweenGusts=7
    maxUpInterpSpeed=5
    minDownInterpSpeed=2
    minGustDuration=0.5
    minGustStrength=3
    minTimeBetweenGusts=0
    minUpInterpSpeed=2
    windTurbulenceScalarA=0.1
    windTurbulenceScalarB=1
    windAmplitude=<0 0 0>
    windFrequency=<0 0 0>
    windOffset=<0 0 0>
   }
   ziplining {
    maxDownInterpSpeed=5
    maxGustDuration=2
    maxGustStrength=7
    maxTimeBetweenGusts=7
    maxUpInterpSpeed=5
    minDownInterpSpeed=2
    minGustDuration=0.5
    minGustStrength=3
    minTimeBetweenGusts=0
    minUpInterpSpeed=2
    windTurbulenceScalarA=0.1
    windTurbulenceScalarB=1
    windAmplitude=<0 0 0>
    windFrequency=<0 0 0>
    windOffset=<0 0 0>
   }
  }
 }
]
