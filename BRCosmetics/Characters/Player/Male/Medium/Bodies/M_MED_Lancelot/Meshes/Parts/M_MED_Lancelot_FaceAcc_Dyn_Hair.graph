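# Physics asset for the dynamic hair on the face accessory of the M_MED_Lancelot player cosmetic.
# Three two-bone hair chains hang off a small dyn_hair_MSTR root body, with one capsule on every bone.
# Swing and twist limits link each chain to the root and each end bone to its parent.
# The first chain swings asymmetrically, narrow on one axis and wide on the other, while the other two use equal swing limits.
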PhysicsAsset /BRCosmetics/Characters/Player/Male/Medium/Bodies/M_MED_Lancelot/Meshes/Parts/M_MED_Lancelot_FaceAcc_Dyn_Hair
bodies {
 dyn_hair_01 [{shapeType=Capsule, length=3, radius=3, shapeName=dyn_hair_01_capsule, center=<3 0 0>, rotation=<rot 90 0 0>}]
 dyn_hair_01_end [{shapeType=Capsule, length=3, radius=3, shapeName=dyn_hair_01_capsule, center=<3 0 0>, rotation=<rot 90 0 0>}]
 dyn_hair_02 [{shapeType=Capsule, length=3, radius=3, shapeName=dyn_hair_02_capsule, center=<3 0 0>, rotation=<rot 90 0 0>}]
 dyn_hair_02_end [{shapeType=Capsule, length=3, radius=3, shapeName=dyn_hair_02_capsule, center=<3 0 0>, rotation=<rot 90 0 0>}]
 dyn_hair_03 [{shapeType=Capsule, length=3, radius=3, shapeName=dyn_hair_03_capsule, center=<3 0 0>, rotation=<rot 90 0 0>}]
 dyn_hair_03_end [{shapeType=Capsule, length=3, radius=3, shapeName=dyn_hair_03_capsule, center=<3 0 0>, rotation=<rot 90 0 0>}]
 dyn_hair_MSTR [{shapeType=Capsule, length=0.505, radius=0.505, shapeName=dyn_hair_MSTR_capsule, rotation=<rot 0 -90 0>}]
}
constraints [
 {bone1Name=dyn_hair_01, bone2Name=dyn_hair_MSTR, swing1LimitDegrees=15, swing2LimitDegrees=60, twistLimitDegrees=15}
 {bone1Name=dyn_hair_02, bone2Name=dyn_hair_MSTR, swing1LimitDegrees=45, swing2LimitDegrees=45, twistLimitDegrees=15}
 {bone1Name=dyn_hair_03, bone2Name=dyn_hair_MSTR, swing1LimitDegrees=45, swing2LimitDegrees=45, twistLimitDegrees=15}
 {bone1Name=dyn_hair_01_end, bone2Name=dyn_hair_01, swing1LimitDegrees=15, swing2LimitDegrees=60, twistLimitDegrees=15}
 {bone1Name=dyn_hair_02_end, bone2Name=dyn_hair_02, swing1LimitDegrees=45, swing2LimitDegrees=45, twistLimitDegrees=15}
 {bone1Name=dyn_hair_03_end, bone2Name=dyn_hair_03, swing1LimitDegrees=45, swing2LimitDegrees=45, twistLimitDegrees=15}
]
